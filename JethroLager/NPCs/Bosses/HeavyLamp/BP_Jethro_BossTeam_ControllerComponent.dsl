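; Controller component for Jethro boss teams that manages a faction tag set and how those factions regard the controller's pawn.
; On pawn spawn it adds the pawn to the factions and applies an attitude, switched to Friendly when the scalable-float toggles allow.
; Converted boss NPCs are mapped to faction tags; an unconvert event removes that tag from the set and unbinds itself.
; Tag additions are gated by a scalable-float toggle.
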(event EventBeginPlay)

(event EventTick
  (params (DeltaSeconds :type Float%20%28single-precision%29)))

(event Custom|AddFactionTag
  (params (GameplayTag :type Gameplay%20Tag%20Structure))
  (if (ScalableFloat|GetValueasBool (Variables|Default|GetCanSetFactions))
    (GameplayTags|AddGameplayTag (Variables|Default|GetFactionTags) GameplayTag :!class /Script/GameplayTags.BlueprintGameplayTagLibrary)))

(event Custom|RemoveFactionTag
  (params (Gameplay%20Tag :type Gameplay%20Tag%20Structure))
  (GameplayTags|RemoveGameplayTag (Variables|Default|GetFactionTags) Gameplay%20Tag :!class /Script/GameplayTags.BlueprintGameplayTagLibrary))

(event EventOnPawnSpawned
  (params (Pawn :type Pawn%20Object%20Reference))
  (Affiliation|AddtoFactions Pawn (Variables|Default|GetFactionTags))
  (if (ScalableFloat|GetValueasBool (Variables|Default|GetCanSetNeutral))
    _
    (elif (ScalableFloat|GetValueasBool (Variables|Default|GetCanSetFriendly))
      (Variables|Default|SetFactionAttitude "Friendly")))
  (Affiliation|ChangeFactionAttitudeTowardsActor (Variables|Default|GetFactionTags) Pawn (Variables|Default|GetFactionAttitude)))

(event Custom|ConenctFactionToConvertedNPC
  (params (BossPawn :type BP%20Player%20Pawn%20Non%20Participant%20Jethro%20Base%20Boss%20Object%20Reference) (FactionTag :type Gameplay%20Tag%20Structure))
  (Utilities|Map|Add (Variables|Default|GetPawnToFactionMap) BossPawn FactionTag)
  (ConvertEvents|BindEventtoOnUnconvertedEvent (Variables|/Script/FortniteGame.FortPlayerPawn|GetConvertComponent BossPawn) (delegate Custom|TeammateNPCUnconverted)))

(event Custom|TeammateNPCUnconverted
  (params (UnconvertedPawn :type Fort%20Pawn%20Object%20Reference) (UnconvertReason :type EUnconvertReason%20Enum :default "StealByOtherPlayer"))
  (bind _pawntofactionmap (Variables|Default|GetPawnToFactionMap))
  (bind (_value _returnvalue) (Utilities|Map|Find _pawntofactionmap UnconvertedPawn))
  (Utilities|FlowControl|Sequence
    (:then_0
      (if _returnvalue
        (CallFunction|RemoveFactionTag _value)))
    (:then_1
      (bind _asbp_player_pawn_non_participant_jethro_base_boss (Utilities|Casting|CastToBP_PlayerPawn_NonParticipant_Jethro_BaseBoss UnconvertedPawn))
      (ConvertEvents|UnbindEventfromOnUnconvertedEvent (Variables|/Script/FortniteGame.FortPlayerPawn|GetConvertComponent _asbp_player_pawn_non_participant_jethro_base_boss) (delegate Custom|TeammateNPCUnconverted)))))
