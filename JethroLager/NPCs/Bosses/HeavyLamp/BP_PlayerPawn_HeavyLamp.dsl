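; Player pawn Blueprint for the HeavyLamp (Vader) boss NPC, covering its fight end and conversion to an ally.
; On DBNO it revives as friendly, kneels, and uses EQS to place a rift gate.
; It leaps along nav path points to the gate, with forced teleport as fallback, then after conversation converts or despawns.
; It also drives walking motion matching, animation target yaw, squad healing on hire and Intercom-synchronized breathing audio.
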(event Patrol|EventOnPatrolEnded
  (bind _self self)
  (bind _returnvalue (Animation|GetAnimationInterface _self))
  (if (not (Variables|Conversion|GetIsConverted))
    ;[1 1 1] Force walking movement style & custom motion matching setup
    (Class|FortPlayerAnimationInterface|SetMovementStyleOverride _returnvalue "Walking")
    (Class|FortPlayerAnimationInterface|SetCustomMotionMatchingDataProxyTable _returnvalue "/JethroLager/NPCs/Bosses/HeavyLamp/Animation/MotionMatchingSetup/HeavyLamp_MotionMatching_PT.HeavyLamp_MotionMatching_PT")))

(event Custom|CallJumpOnTimer
  (Utilities|Time|SetTimerbyEvent (EventDispatchers|CreateEvent :!event JumpToPathLocation) 1.5))

(event EventOnCharacterCustomizationCompleted
  (params (Pawn :type Fort%20Player%20Pawn%20Object%20Reference))
  (bind _self self)
  (bind _returnvalue (Animation|GetAnimationInterface _self))
  (if (not (Variables|Conversion|GetIsConverted))
    ;[1 1 1] Force walking movement style & custom motion matching setup
    (Class|FortPlayerAnimationInterface|SetMovementStyleOverride _returnvalue "Walking")
    (Class|FortPlayerAnimationInterface|SetCustomMotionMatchingDataProxyTable _returnvalue "/JethroLager/NPCs/Bosses/HeavyLamp/Animation/MotionMatchingSetup/HeavyLamp_MotionMatching_PT.HeavyLamp_MotionMatching_PT")))

(event Custom|Request%20Breathing
  ; Send requests to Intercom manager to start and stop global breathing sound instance
  (GameplayEvent|BroadcastEvent (Variables|Sound|GetBreathing%20Broadcast%20Request%20Event) 0 (Utilities|Struct|MakeGameplayEventIntercomBroadcastRequest (Variables|Sound|GetBreathing%20Channel) (Variables|Sound|GetBreathing%20Sound) 0 true false) "/Script/FortniteGame.FortGameState"))

(event Custom|HeavyLamp_BeginPlay
  (EventDispatchers|AssignOnPlayerStartDBNO (EventDispatchers|CreateEvent :!event Custom|OnHeavyLampStartDBNO))
  (EventDispatchers|BindEventtoOnDamaged (EventDispatchers|CreateEvent :!event HeavyLamp_OnDamaged))
  (Variables|Default|SetWasForceTeleported false)
  (Networking|CanExecuteCosmeticEvents
    (:True
      (Default|CallOnHeavyLampConverted (Variables|Conversion|GetIsConverted))
      (Default|CallOnBossFightFinished (Variables|RiftTransition|GetBossFightFinished))
      ;[0.15 0.15 0.15 0.5 16] Plays breathing sound using Intercom global sound instance manager to synchronize breaths across multiple emitters
      (Class|IntercomSpeakerComponent|StartListening (Variables|Default|GetIntercomSpeakerComponent_Breathing))
      (CallFunction|RequestBreathing))))

(event Custom|Stop%20Breathing
  (params (Fade%20Out%20Duration :type Float%20%28double-precision%29))
  ; Send requests to Intercom manager to start and stop global breathing sound instance
  (GameplayEvent|BroadcastEvent (Variables|Sound|GetBreathing%20Broadcast%20Stop%20Event) 0 (Utilities|Struct|MakeGameplayEventIntercomStopBroadcast (Variables|Sound|GetBreathing%20Channel) Fade%20Out%20Duration) "/Script/FortniteGame.FortGameState"))

(event Custom|OnHeavyLampStartDBNO
  (EventDispatchers|UnbindEventfromOnPlayerStartDBNO (EventDispatchers|CreateEvent :!event Custom|OnHeavyLampStartDBNO))
  (Networking|SwitchHasAuthority
    (:Authority
      (bind _returnvalue (Ability|GameplayCue|IsGameplayCueActive self (Variables|Default|GetLeapInAirCue)))
      (if (not _returnvalue)
        (CallFunction|HeavyLampFinishedBossFight)
        (else
          (bind _returnvalue_2 (Utilities|Time|SetTimerbyEvent (EventDispatchers|CreateEvent :!event Custom|PostLeapDBNO) 0.1 true true))
          (Variables|Default|SetLeapCheckTimer _returnvalue_2))))))

(event EventEnableHealthBarDelayedEvent)

(event Custom|PostLeapDBNO
  (bind _returnvalue (Ability|GameplayCue|IsGameplayCueActive self (Variables|Default|GetLeapInAirCue)))
  (if (and (not _returnvalue) (Utilities|Time|IsValidTimerHandle (Variables|Default|GetLeapCheckTimer)))
    (Utilities|Time|ClearandInvalidateTimerbyHandle (Variables|Default|GetLeapCheckTimer))
    (CallFunction|HeavyLampFinishedBossFight)))

(event Custom|FallbackForceTeleport
  ; If Vader did not reach the Rift Location by himself in the amount of time allowed, force him to be there.
  (if (not (> (Math|Vector|Distance2D(Vector) (Transformation|GetActorLocation) (Variables|RiftGate|GetUpdateToQueryWorldPosition)) (ScalableFloat|GetValueatLevel (Variables|Default|GetForceTeleportFallbackDistance))))
    (Utilities|IsValid (Variables|Default|GetSpawnedConversationActor)
      (:Is%20Not%20Valid
        _)))
  (Tags|CancelAbilitiesWithTags (Ability|GetFortAbilitySystemComponent self) (Variables|Default|GetTagsForAbilityCancels))
  (CallFunction|ForceRiftGatePosition))

(event Custom|HeavyLampFinishedBossFight
  (OnDBNO|HandleConvertfromDBNO)
  ;[0.8438 0.8438 0.8438 0.5] Helps clean montages from left over abilities
  (Utilities|FlowControl|DelayUntilNextTick)
  (bind _animwithkneelduration (OnDBNO|PlayKneelAnim (Variables|RiftTransition|GetKneel%20Duration)))
  (Utilities|Time|SetTimerbyEvent (EventDispatchers|CreateEvent :!event Custom|OnKneelFinished) _animwithkneelduration)
  (OnDBNO|ApplyQuestVerbtoDefeatingTeam)
  (if (ScalableFloat|GetValueasBool (Variables|RiftGate|GetCanSpawnRiftGate))
    (RiftGate|CheckLocationforRiftGateSpawn)))

(event Custom|OnKneelFinished
  (bind _animduration (OnDBNO|PlayStandupAnim))
  (Utilities|Time|SetTimerbyEvent (EventDispatchers|CreateEvent :!event Custom|OnStandupFinished) _animduration))

(event Custom|ForceRiftGatePosition
  (bind _forceteleportlocation (Variables|Default|GetForceTeleportLocation))
  (bind _spawnedriftgate (Variables|RiftGate|GetSpawnedRiftGate))
  (bind _spawnedconversationactor (Variables|Default|GetSpawnedConversationActor))
  ; Force Vader to go to Rift Gate Location by Teleporting Him there.
  (Variables|Default|SetWasForceTeleported true)
  (bind _asfort_athena_aibot_controller (Utilities|Casting|CastToFortAthenaAIBotController (Pawn|GetController self)))
  (AI|PushBehaviorControl _asfort_athena_aibot_controller -4194272)
  (CallFunction|TriggerRiftVFX (Transformation|GetActorLocation))
  (bind (_outhit _returnvalue) (Collision|LineTraceByChannel (+ (Variables|RiftGate|GetUpdateToQueryWorldPosition) "0.000000,0.000000,200.000000") (- (Variables|RiftGate|GetUpdateToQueryWorldPosition) "0.000000,0.000000,1500.000000") "TraceTypeQuery7" false (Utilities|Array|MakeArray _spawnedriftgate)))
  (if _returnvalue
    (Variables|Default|SetForceTeleportLocation (Utilities|Operators|Add (.Location (Collision|BreakHitResult _outhit)) (Math|Vector|MakeVector 0.0 0.0 (Components|Capsule|GetScaledCapsuleHalfHeight (Variables|Character|GetCapsuleComponent)))))
    (else
      (Variables|Default|SetForceTeleportLocation (Variables|RiftGate|GetUpdateToQueryWorldPosition))))
  (if (.ReturnValue (AI|Navigation|ProjectPointtoNavigation _forceteleportlocation (Fortnite|AI|GetNavigationDataforActor _asfort_athena_aibot_controller) "/Game/Athena/AI/Phoebe/NavFilter/NavFilter_Phoebe_NoSmashableNoWater.NavFilter_Phoebe_NoSmashableNoWater_C" "500.000000,500.000000,200.000000"))
    (bind _output_get (Variables|Default|SetForceTeleportLocation (AI|Navigation|ProjectPointtoNavigation _forceteleportlocation (Fortnite|AI|GetNavigationDataforActor _asfort_athena_aibot_controller) "/Game/Athena/AI/Phoebe/NavFilter/NavFilter_Phoebe_NoSmashableNoWater.NavFilter_Phoebe_NoSmashableNoWater_C" "500.000000,500.000000,200.000000")))
    (Variables|Default|SetForceTeleportLocation (Utilities|Operators|Add _output_get (Math|Vector|MakeVector 0.0 0.0 (Components|Capsule|GetScaledCapsuleHalfHeight (Variables|Character|GetCapsuleComponent)))))
    (else
      _))
  (Transformation|SetActorLocationAndRotation (Variables|Default|GetForceTeleportLocation) (Transformation|GetActorRotation _spawnedriftgate) false true)
  (Utilities|IsValid _spawnedconversationactor
    (:Is%20Valid
      (Class|BGAAthenaPlotTwistHeavyLampConvo|ForceRemove _spawnedconversationactor))
    (:Is%20Not%20Valid
      _))
  (Utilities|FlowControl|Delay 1.0)
  (RiftGate|SetupConversationBGA (ScalableFloat|GetValueasBool (Variables|Default|GetConvoUseTeleportLocation))))

(event Custom|OnStandupFinished
  ; Vader is getting up from Kneeling
  (Components|Activation|Activate (Variables|Character|GetCharacterMovement) false)
  (if (not (ScalableFloat|GetValueasBool (Variables|RiftGate|GetCanSpawnRiftGate)))
    (if (and (Variables|Conversion|GetValidLastDamagingPlayer) (Convert|GetCanbeConverted (Variables|FortPlayerPawn|GetConvertComponent)))
      (Variables|Conversion|SetConverterPawn (Variables|Conversion|GetLastDamagingPawn))
      (CallFunction|HeavyLampConvert)
      (else
        (CallFunction|StartHeavyLampDespawn)))
    (elif (ScalableFloat|GetValueasBool (Variables|Default|GetForceTeleportAfterReviveFromDBNO))
      (CallFunction|ForceRiftGatePosition))))

(event Custom|TriggerRiftVFX
  (params (Location :type Vector))
  ; Force Vader to go to Rift Gate Location by Teleporting Him there.
  (Niagara|SpawnSystematLocation "/PlotTwist/Items/RiftGate/Assets/FX/NS_PlotTwist_Portal_EntryExit.NS_PlotTwist_Portal_EntryExit" Location))

(event EventOnDamageServer
  (params (Damage :type Float%20%28single-precision%29) (DamageTags :type Gameplay%20Tag%20Container%20Structure%20%28by%20ref%29) (Momentum :type Vector :default "0, 0, 0") (HitInfo :type Hit%20Result%20Structure%20%28by%20ref%29) (InstigatedBy :type Controller%20Object%20Reference) (DamageCauser :type Actor%20Object%20Reference) (EffectContext :type Gameplay%20Effect%20Context%20Handle%20Structure))
  (Variables|Default|SetLastDamagerPlayerController InstigatedBy))

(event EventBeginPlay
  ;[1 1 1] Check to know if the AI has a Weapon Equipped or not at the time of death (before he actually dies - when he dies, it removes the weapon). Set a variable that will be used in the Gamepaly Cue Rezzout to validate which animation should play.
  (|CallParentFunction)
  (FortPawn|BindEventtoOnWeaponUnEquippedDelegate (delegate Custom|WeaponUnequipped))
  (FortPawn|BindEventtoOnWeaponEquippedDelegate (delegate Custom|WeaponEquipped))
  (CallFunction|HeavyLampBeginPlay)
  (Networking|SwitchHasAuthority
    (:Authority
      (Utilities|FlowControl|Sequence
        (:then_0
          ;[1 1 1] To Be Used in the Ability where Realm needs to Jump away from Players if Realm and Player is in the Water.
          (Variables|Default|SetSpawnLocation (Transformation|GetActorLocation self))
          (bind _returnvalue (Game|SpawnActorfromClass "/Realm/Items/BP_Realm_StartLocation.BP_Realm_StartLocation_C" (Transformation|GetActorTransform self) "AlwaysSpawn" "OverrideRootScale" self self))
          (Variables|Default|SetRealmOriginActor _returnvalue)
          (Utilities|FlowControl|Delay 1.0) ; This 1s delay is needed, else the Blackboard doesn't seem ready to receive a Value as Object
          (AI|Components|Blackboard|SetValueasObject (AI|GetBlackboard (AI|GetAIController self)) (Variables|Default|GetBBKEY_OriginActor) (Variables|Default|GetRealmOriginActor))
          (Variables|Default|SetCanEnableHealthBars false))
        (:then_1
          ; Logic to update the Target Yaw so Vader can look at his target.
          (bind _asfort_athena_aibot_controller (Utilities|Casting|CastToFortAthenaAIBotController (Pawn|GetController self)))
          (Perception|BindEventtoOnCurrentTargetChangedEvent _asfort_athena_aibot_controller (delegate Custom|CurrentTargetChanged)))
        (:then_2
          ; Update Map Marker after a Delay
          (bind _returnvalue_2 (Utilities|Time|SetTimerbyEvent (delegate Custom|OnDisplayMapMarkerDelayEnded) (ScalableFloat|GetValueatLevel (Variables|Default|GetMapMarkerActivationTime))))
          (Variables|Default|SetDisplayMapMarkerTimerHandle _returnvalue_2))))))

(event Custom|WeaponUnequipped
  ;[1 1 1] Check to know if the AI has a Weapon Equipped or not at the time of death (before he actually dies - when he dies, it removes the weapon). Set a variable that will be used in the Gamepaly Cue Rezzout to validate which animation should play.
  (if (not (FortPawn|IsDead self))
    (Variables|Default|SetWeaponHolsteredOnDeath true)))

(event Custom|WeaponEquipped
  (params (NewWeapon :type Fort%20Weapon%20Object%20Reference) (PrevWeapon :type Fort%20Weapon%20Object%20Reference))
  ;[1 1 1] Check to know if the AI has a Weapon Equipped or not at the time of death (before he actually dies - when he dies, it removes the weapon). Set a variable that will be used in the Gamepaly Cue Rezzout to validate which animation should play.
  (if (not (FortPawn|IsDead self))
    (Variables|Default|SetWeaponHolsteredOnDeath false)))

(event Custom|RealmIsDeflecting
  ;[1 1 1] Realm is Deflecting - Specific Bool for Anim Tree for Animators to get Vader to Hold a Pose
  (Variables|Default|SetIsCurrentlyDeflecting true)
  (Networking|ForceNetUpdate)
  (Utilities|FlowControl|RetriggerableDelay (ScalableFloat|GetValueatLevel (Variables|Default|GetDeflectingAnimDelay)))
  (Variables|Default|SetIsCurrentlyDeflecting false)
  (Networking|ForceNetUpdate))

(event Custom|CurrentTargetChanged
  (params (OldTarget :type Actor%20Object%20Reference) (NewTarget :type Actor%20Object%20Reference))
  ; Logic to update the Target Yaw so Vader can look at his target.
  (Variables|Default|SetAITarget NewTarget))

(event Custom|OnDisplayMapMarkerDelayEnded
  ; Update Map Marker after a Delay
  (Components|Activation|SetActive (Variables|Default|GetHeavyLamp_SpecialActor) true false))

(event EventOnDeathServer
  (params (Damage :type Float%20%28single-precision%29) (DamageTags :type Gameplay%20Tag%20Container%20Structure%20%28by%20ref%29) (Momentum :type Vector :default "0, 0, 0") (HitInfo :type Hit%20Result%20Structure%20%28by%20ref%29) (InstigatedBy :type Controller%20Object%20Reference) (DamageCauser :type Actor%20Object%20Reference) (EffectContext :type Gameplay%20Effect%20Context%20Handle%20Structure))
  (bind _realmoriginactor (Variables|Default|GetRealmOriginActor))
  ;[1 1 1] On Death Server - Destroy the Origin Actor we Spawned
  (Utilities|IsValid _realmoriginactor
    (:Is%20Valid
      (Actor|DestroyActor _realmoriginactor))))

(event Custom|StartYawUpdateTimer
  ;[1 1 1] Check if Pawn has a Target and get the Angle Difference and setit in a Variabl so Animators can use the value.
  (bind _returnvalue (Utilities|Time|SetTimerbyEvent (EventDispatchers|CreateEvent :!event Custom|CheckPawnTargetPosition) 0.1 true))
  (Variables|Default|SetHNDL_UpdateTargettingYaw _returnvalue))

(event Custom|CheckPawnTargetPosition
  (bind _aitarget (Variables|Default|GetAITarget))
  ;[1 1 1] Check if Pawn has a Target and get the Angle Difference and setit in a Variabl so Animators can use the value.
  (Utilities|IsValid _aitarget
    (:Is%20Valid
      (Variables|Animation|SetTargetYaw (select (< (Math|Vector|DotProduct (Transformation|GetActorRightVector self) (Math|Vector|Normalize (- (Transformation|GetActorLocation _aitarget) (Transformation|GetActorLocation self)))) 0) (Class|AbilityFunctions|GetAngleBetweenTwoVectors (Math|Vector|Normalize (- (Transformation|GetActorLocation _aitarget) (Transformation|GetActorLocation self))) (Transformation|GetActorForwardVector self)) (* (Class|AbilityFunctions|GetAngleBetweenTwoVectors (Math|Vector|Normalize (- (Transformation|GetActorLocation _aitarget) (Transformation|GetActorLocation self))) (Transformation|GetActorForwardVector self)) -1.0))))
    (:Is%20Not%20Valid
      (Variables|Animation|SetTargetYaw 0.0) ; If no Target, set the Yaw at 0
    )))

(event OnUnconvertedEvent(ConvertComponent)
  (params (UnconvertedPawn :type Fort%20Pawn%20Object%20Reference) (UnconvertReason :type EUnconvertReason%20Enum :default "StealByOtherPlayer"))
  (bind _self self)
  (bind _returnvalue (Actor|GetComponentbyClass _self "/Game/Sounds/Fort_Audio_Blueprints/VoiceComponent/NPC_VoiceComponent.NPC_VoiceComponent_C"))
  ; On Unconvert, allow to show back Health Bar if Vader turns against his converter.
  (Conversion|HandlePawnUnconverted UnconvertedPawn)
  (if (Utilities|Enum|Equal(Enum) UnconvertReason "TurnAgainstConverter")
    (Variables|Default|SetCanEnableHealthBars true)
    (Variables|Default|SetEnableHealthBar true)
    (CallFunction|NPCEnableHealthBars)
    (Variables|Default|SetNPC%20HealthBarHideTimeAfterDeath 1.2)
    (elif (Utilities|Enum|Equal(Enum) UnconvertReason "ConverterDeath")
      ;[0.15 0.05847 0.1478 0.5] Ensure 2D squad-only death VO doesn't play on squad wipe
      (Utilities|IsValid _returnvalue
        (:Is%20Valid
          _))))
  (Variables|/Script/FortniteAI.FortAIComponent_Voice|SetVoiceSoundBank 0 _returnvalue))

(event Custom|RiftGateOutHeavyLampNPC
  ;[1 0.1863 0.1094 0.5] Handle Vader post-rift gate to remove from map
  (bind _spawnedconversationactor (Variables|Default|GetSpawnedConversationActor)
    (:then
      (Variables|Conversion|SetConverterPawn (Variables|/PlotTwist/Items/RiftGate/RiftGate_HeavyLamp/BGA_Athena_PlotTwist_HeavyLamp_Convo.BGA_Athena_PlotTwist_HeavyLamp_Convo_C|GetBeingGrantedPowersPawn _spawnedconversationactor))
      (Default|UnbindEventfromTrainingAcceptedbyConversation _spawnedconversationactor (EventDispatchers|CreateEvent :!event Custom|BeginRiftTrainingSequence))
      (Default|UnbindEventfromForcedEndOfTraining _spawnedconversationactor (EventDispatchers|CreateEvent :!event Custom|ExitWithoutTraining))
      (Class|BGAAthenaPlotTwistHeavyLampConvo|ForceRemove _spawnedconversationactor))
    (:else
      _))
  (if (Convert|GetCanbeConverted (Variables|FortPlayerPawn|GetConvertComponent))
    (bind _returnvalue (Utilities|Time|SetTimerbyEvent (EventDispatchers|CreateEvent :!event Custom|HeavyLamp_Convert) (ScalableFloat|GetValueatLevel (Variables|RiftOut|GetVaderActAfterRiftTime))))
    (Variables|Conversion|SetHeavyLampConvertTimer _returnvalue)
    (Default|BindEventtoOnRiftGateSequenceFinished (Variables|RiftGate|GetSpawnedRiftGate) (delegate Custom|OnRiftGateSequenceFinished))
    (else
      (bind _returnvalue_2 (Utilities|Time|SetTimerbyEvent (EventDispatchers|CreateEvent :!event Custom|StartHeavyLampDespawn) (Variables|RiftOut|GetVaderLeaveAfterRiftTime)))
      (Variables|RiftOut|SetHeavyLampDestroyTimer _returnvalue_2)
      (CallFunction|SpawnRiftGateSmartObject))))

(event Custom|OnRiftGateSequenceFinished
  ;[1 0.1863 0.1094 0.5] Handle Vader post-rift gate to remove from map
  (CallFunction|HeavyLampConvert))

(event Custom|StartHeavyLampDespawn
  ;[1 0.1863 0.1094 0.5] Handle Vader post-rift gate to remove from map
  (Utilities|Time|ClearandInvalidateTimerbyHandle (Variables|RiftOut|GetHeavyLampDestroyTimer))
  (CallFunction|TriggerRiftVFX (Transformation|GetActorLocation self))
  (Variables|Default|SetSetInvisible true)
  (CallFunction|HeavyLampDespawn))

(event Custom|HeavyLampDespawn
  ;[1 0.1863 0.1094 0.5] Handle Vader post-rift gate to remove from map
  (FortPawn|ForceKill "(TagName=\"Gameplay.Effect.InstantDeath.Silent\")"))

(event Custom|SpawnRiftGateSmartObject
  ;[1 0.1863 0.1094 0.5] Handle Vader post-rift gate to remove from map
  (bind _spawnedriftgate (Variables|RiftGate|GetSpawnedRiftGate)
    (:then
      (Components|Activation|Deactivate (Variables|Character|GetCharacterMovement))
      (bind (_returnvalue_location _returnvalue_rotation) (Math|Transform|BreakTransform (Transformation|GetActorTransform _spawnedriftgate)))
      (bind _returnvalue (Game|SpawnActorfromClass "/JethroLager/SmartObjects/HeavyLamp/SmartObject_NoMesh_Actor_HeavyLamp.SmartObject_NoMesh_Actor_HeavyLamp_C" (Math|Transform|MakeTransform _returnvalue_location _returnvalue_rotation) "Undefined" "MultiplyWithRoot" 0 self))
      (bind _pathsmartobject (Variables|RiftTransition|GetPathSmartObject)
        (:then
          (Actor|DestroyActor _pathsmartobject))
        (:else
          _))
      (Variables|RiftTransition|SetPathSmartObject _returnvalue))
    (:else)))

(event Custom|BeginRiftTrainingSequence
  ;[1 0.1863 0.1094 0.5] Handle Vader post-rift gate to remove from map
  (bind _spawnedconversationactor (Variables|Default|GetSpawnedConversationActor)
    (:then
      (bind _spawnedriftgate (Variables|RiftGate|GetSpawnedRiftGate)
        (:then
          (Variables|/Game/Athena/Items/EnvironmentalItems/HidingProps/B_HidingProp.B_HidingProp_C|SetLastPawnToInteract (Variables|/PlotTwist/Items/RiftGate/RiftGate_HeavyLamp/BGA_Athena_PlotTwist_HeavyLamp_Convo.BGA_Athena_PlotTwist_HeavyLamp_Convo_C|GetBeingGrantedPowersPawn _spawnedconversationactor) _spawnedriftgate)
          (Class|BAthenaPlotTwistRiftGateParent|GainForcePowers _spawnedriftgate)
          (Variables|Default|SetbShouldRiftOut true)
          (Utilities|Time|ClearandInvalidateTimerbyHandle (Variables|Default|GetFallbackForceTeleportTimerHandle))
          (CallFunction|RiftGateOutHeavyLampNPC))
        (:else)))
    (:else)))

(event Custom|ExitWithoutTraining
  ;[1 0.1863 0.1094 0.5] Handle Vader post-rift gate to remove from map
  (Variables|Default|SetbShouldRiftOut true)
  (Utilities|Time|ClearandInvalidateTimerbyHandle (Variables|Default|GetFallbackForceTeleportTimerHandle))
  (CallFunction|RiftGateOutHeavyLampNPC))

(event Custom|HeavyLamp_Convert
  ; Set Heavy Lamp to be Converted
  (if (Variables|Default|GetWasForceTeleported)
    (bind _asfort_athena_aibot_controller (Utilities|Casting|CastToFortAthenaAIBotController (Pawn|GetController self))
      (:then
        (AI|PopBehaviorControl _asfort_athena_aibot_controller))
      (:CastFailed
        _))
    (Utilities|Time|ClearandInvalidateTimerbyHandle (Variables|Conversion|GetHeavyLampConvertTimer))
    (Components|Activation|Activate (Variables|Character|GetCharacterMovement) false)
    (Convert|TriggerConversion (Variables|Conversion|GetConverterPawn) self)
    (Fortnite|SetCanbeDamaged self true)
    (Default|UnbindEventfromOnRiftGateSequenceFinished (Variables|RiftGate|GetSpawnedRiftGate) (EventDispatchers|CreateEvent :!event Custom|OnRiftGateSequenceFinished))
    (bind _pathsmartobject (Variables|RiftTransition|GetPathSmartObject)
      (:then
        (Actor|DestroyActor _pathsmartobject))
      (:else))
    (else
      _)))

(event EventEndPlay
  (params (EndPlayReason :type EEndPlayReason%20Enum :default "Destroyed"))
  (bind _pathsmartobject (Variables|RiftTransition|GetPathSmartObject)
    (:then
      (Actor|DestroyActor _pathsmartobject))
    (:else)))

(event Custom|Play%202D%20Death%20VO%20for%20Team
  (params (Cause :type EDeathCause%20Enum :default "OutsideSafeZone"))
  (bind _returnvalue (Audio|SpawnSound2D (Variables|Sound|GetDeath%20VO)))
  (Audio|Components|Audio|SetSubmixSend _returnvalue "/Game/Sounds/SubMixes/Fortnite_SFX.Fortnite_SFX" 0.0 :!class /Script/Engine.AudioComponent)
  (Audio|Components|Audio|SetSubmixSend _returnvalue "/Game/Sounds/SubMixes/Fortnite_Chat.Fortnite_Chat" 1.0 :!class /Script/Engine.AudioComponent)
  (EventDispatchers|BindEventtoOnAudioFinished _returnvalue (delegate Custom|Death%20VO%20Finished)))

(event Custom|Death%20VO%20Finished
  (Class|IntercomSpeakerComponent|StopListening (Variables|Default|GetIntercomSpeakerComponent_Breathing))
  (CallFunction|StopBreathing 0.2))

(fn ActorDistanceToPoint
  (params (Point :type Vector))
  (return (Math|Vector|Distance(Vector) Point (Transformation|GetActorLocation self))))

(fn AddJethroControllerComponentToBossTeam
  (bind _lonefeatherfaction (Variables|Conversion|GetLoneFeatherFaction))
  (for _array_element (Variables|Conversion|GetConvertedSquadPawns)
    (bind _returnvalue (AddComponent|NODE_AddBP_Jethro_BossTeam_ControllerComponent-1 :self (Fortnite|GetFortPlayerControllerfromActor _array_element)))
    (Class|BPJethroBossTeamControllerComponent|AddFactionTag _returnvalue _lonefeatherfaction)
    (Class|BPJethroBossTeamControllerComponent|ConenctFactiontoConvertedNPC _returnvalue self _lonefeatherfaction)))

(fn ApplyQuestVerbToDefeatingTeam
  ;[0.9307 0.3333 1 0.5] Get defeating players squad to apply verb
  (bind _lastdamagingpawn (Variables|Conversion|GetLastDamagingPawn)
    (:then
      (bind _asfort_player_state_athena (Utilities|Casting|CastToFortPlayerStateAthena (Variables|/Script/Engine.Pawn|GetPlayerState _lastdamagingpawn)))
      (bind _returnvalue (Class|FortGameStateAthena|GetSquadMemberPlayerStates (Fortnite|Athena|GameState|GetGameStateAthenafromContext (Fortnite|GetFortPlayerControllerfromActor _lastdamagingpawn)) _asfort_player_state_athena true))
      (for _array_element _returnvalue
        (bind _asfort_player_pawn (Utilities|Casting|CastToFortPlayerPawn (PlayerState|GetPawn _array_element)))
        (CustomVerb|BroadcastCustomVerbMessage (CustomVerb|MakeCustomVerbMessage _asfort_player_pawn "(TagName=\"CustomVerb.HeavyLampDown\")"))))
    (:else)))

(fn AssignLastDamagingPawn
  (params (DamagingPawn :type Fort%20Player%20Pawn%20Object%20Reference))
  (if (Pawn|IsPlayerControlled DamagingPawn)
    (Variables|Conversion|SetLastDamagingPawn DamagingPawn)
    (Variables|Conversion|SetValidLastDamagingPlayer true)
    (else
      (Variables|Conversion|SetValidLastDamagingPlayer false))))

(fn BlueprintModifyIncomingDamage
  (params (Damage :type Float%20%28single-precision%29) (DamageTags :type Gameplay%20Tag%20Container%20Structure%20%28by%20ref%29) (EffectContext :type Gameplay%20Effect%20Context%20Handle%20Structure%20%28by%20ref%29) (EventInstigator :type Controller%20Object%20Reference) (DamageCauser :type Actor%20Object%20Reference) (HitInfo :type Hit%20Result%20Structure))
  (bind _returnvalue (|CallParentFunction Damage DamageTags EffectContext EventInstigator DamageCauser HitInfo))
  (Utilities|IsValid EventInstigator
    (:Is%20Valid
      (if (!= EventInstigator (Pawn|GetController))
        (bind _asfort_pawn (Utilities|Casting|CastToFortPawn (Pawn|GetControlledPawn EventInstigator))
          (:then
            (if (== (CallFunction|GetTeam _asfort_pawn) (CallFunction|GetTeam))
              (return 1e-05)
              (else
                _)))
          (:CastFailed
            _))
        (return _returnvalue)
        (else
          _)))
    (:Is%20Not%20Valid
      _)))

(fn CheckLocationForRiftGateSpawn
  (bind _returnvalue (AI|EQS|RunEQSQuery (Variables|RiftGate|GetEQSQueryTemplate) self "AllMatching"))
  (Utilities|IsValid _returnvalue
    (:Is%20Valid
      (AI|EQS|BindEventtoOnQueryFinished _returnvalue (EventDispatchers|CreateEvent :!event SetSpawnerAfterEQSQueryComplete)))))

(fn CheckRiftPathReady
  (bind _riftpathready (Variables|RiftGate|GetRiftPathReady))
  (if _riftpathready
    (return _riftpathready)
    (elif (and (not (Variables|RiftGate|GetNoValidRiftLocations)) (> (Variables|RiftGate|GetRiftPathReadyRetries) (Variables|RiftGate|GetMaxRiftPathReadyRetries)))
      (Math|Integer|IncrementInt (Variables|RiftGate|GetRiftPathReadyRetries)
        (:%20%20
          (Utilities|Time|SetTimerbyEvent (EventDispatchers|CreateEvent :!event OnStandupFinished) 0.2)
          (return false)))
      (else
        (RiftGate|CheckLocationforRiftGateSpawn)))))

(fn DivideVector
  (params (DivideValue :type Float%20%28double-precision%29) (StartVector :type Vector))
  (return (+ (Utilities|Operators|Divide (- StartVector (Transformation|GetActorLocation self)) (Math|Vector|MakeVector DivideValue DivideValue DivideValue)) (Transformation|GetActorLocation self))))

(fn FindRiftPath
  (Networking|SwitchHasAuthority
    (:Authority
      (if (not (ScalableFloat|GetValueasBool (Variables|Default|GetForceTeleportAfterReviveFromDBNO)))
        (bind _asfort_athena_aibot_controller (Utilities|Casting|CastToFortAthenaAIBotController (AI|GetAIController self)))
        (bind _returnvalue (AI|Navigation|FindPathtoLocationSynchronously (Transformation|GetActorLocation self) (Variables|RiftGate|GetUpdateToQueryWorldPosition) _asfort_athena_aibot_controller))
        (bind _output_get (Variables|RiftTransition|SetPath%20Points (Variables|/Script/NavigationSystem.NavigationPath|GetPathPoints _returnvalue)))
        (Utilities|Array|RemoveIndex _output_get)
        (Variables|RiftGate|SetRiftPathReady true)
        (RiftTransition|StartPathtoRift)))))

(fn GetAnimData
  (return (Variables|Animation|GetTargetYaw) (Variables|Default|GetIsCurrentlyDeflecting)))

(fn GetBestRiftGateLocationFromEQSResults
  (params (EQSQueryResults :type EQS%20Query%20Instance%20Object%20Reference))
  (bind _h_currentbestlocation (LocalVariables|Geth_currentBestLocation))
  (bind _self self)
  (bind _returnvalue (Transformation|GetActorLocation _self))
  (bind _returnvalue_2 (Math|Vector|Distance2DSquared(Vector) _returnvalue _h_currentbestlocation))
  (bind _lastdamagerplayercontroller (Variables|Default|GetLastDamagerPlayerController)
    (:then
      (Utilities|IsValid (Pawn|GetControlledPawn _lastdamagerplayercontroller)
        (:Is%20Valid
          (LocalVariables|Seth_PawnThatEliminatedMe (Pawn|GetControlledPawn _lastdamagerplayercontroller))
          (bind (_resultlocations _returnvalue_3) (AI|EQS|GetQueryResultsasLocations EQSQueryResults)) ; will return all valid locations
          (if _returnvalue_3
            (for (_array_element _array_index) _resultlocations
              (if (> _array_index 0)
                (bind _h_pawnthateliminatedme (LocalVariables|Geth_PawnThatEliminatedMe)
                  (:then
                    (if (< (Math|Vector|DotProduct (- (Transformation|GetActorLocation self) (Transformation|GetActorLocation _h_pawnthateliminatedme)) (- (Transformation|GetActorLocation self) _array_element)) 0)
                      (if (LocalVariables|Geth_bBestResultUsedDirection)
                        (if (< (Math|Vector|Distance2DSquared(Vector) _array_element _returnvalue) _returnvalue_2)
                          _)
                        (else
                          _))
                      (LocalVariables|Seth_currentBestLocation _array_element)
                      (LocalVariables|Seth_bBestResultUsedDirection true)
                      (elif (LocalVariables|Geth_bBestResultUsedDirection)
                        _)))
                  (:else
                    _)) ; we don't want to fail if we can't validate a player that eliminated HeavyLamp, \nbut we will want to adjust how we determine the best location:\nonly use distance, otherwise, choose a location "behind" Vader to better frame the visuals
                (if (< (Math|Vector|Distance2DSquared(Vector) _array_element _returnvalue) _returnvalue_2)
                  (LocalVariables|Seth_currentBestLocation _array_element))
                (else
                  (bind _h_pawnthateliminatedme_2 (LocalVariables|Geth_PawnThatEliminatedMe)
                    (:then
                      (LocalVariables|Seth_bBestResultUsedDirection (< (Math|Vector|DotProduct (- (Transformation|GetActorLocation self) (Transformation|GetActorLocation _h_pawnthateliminatedme_2)) (- (Transformation|GetActorLocation self) _array_element)) 0)) ; need to know if this used best direction
                    )
                    (:else
                      _)) ; we don't want to fail if we can't validate a player that eliminated HeavyLamp, \nbut we will want to adjust how we determine the best location:\nonly use distance, otherwise, choose a location "behind" Vader to better frame the visuals
                  (LocalVariables|Seth_currentBestLocation _array_element) ; no result yet, so store the first index
                )))
            (return (LocalVariables|Geth_currentBestLocation) true)
            (else
              (return "0, 0, 0" false))))))
    (:else)) ; will try to use the eliminating player \nif possible to better frame the rift placement
)

(fn GetSquadPawns
  (bind _convertcomponent (Variables|FortPlayerPawn|GetConvertComponent))
  (bind _returnvalue (Convert|GetConverterPawn _convertcomponent))
  (Utilities|IsValid _returnvalue
    (:Is%20Valid
      (Utilities|Array|AddUnique (Variables|Conversion|GetConvertedSquadPawns) _returnvalue)
      ;[0.9307 0.3333 1 0.5] Get converting players squad from player state
      (bind _asfort_player_state_athena (Utilities|Casting|CastToFortPlayerStateAthena (Variables|/Script/Engine.Pawn|GetPlayerState _returnvalue)))
      (bind _returnvalue_2 (Class|FortGameStateAthena|GetSquadMemberPlayerStates (Fortnite|Athena|GameState|GetGameStateAthenafromContext (Fortnite|GetFortPlayerControllerfromActor _returnvalue)) _asfort_player_state_athena false))
      (for _array_element _returnvalue_2
        (bind _asfort_player_pawn (Utilities|Casting|CastToFortPlayerPawn (PlayerState|GetPawn _array_element)))
        (Utilities|Array|AddUnique (Variables|Conversion|GetConvertedSquadPawns) _asfort_player_pawn)))))

(fn HandleConvertFromDBNO
  (bind _self self)
  (bind _returnvalue (Ability|GetFortAbilitySystemComponent _self))
  (bind _returnvalue_2 (Class|AthenaFunctionLibrary|MakeSimpleFortEffectContext _self))
  (bind _self_2 self)
  (Variables|RiftTransition|SetBossFightFinished true)
  (Fortnite|SetCanbeDamaged self)
  (Components|Activation|Deactivate (Variables|Character|GetCharacterMovement))
  (CallFunction|ForceRevivefromDBNO 0)
  (Affiliation|RemovefromAllFactions _self_2)
  (Affiliation|AddtoFaction _self_2 (Variables|RiftTransition|GetHeavyLampFriendlyFaction))
  (GameplayEffects|ApplyGameplayEffectToSelf _returnvalue "/JethroLager/NPCs/Bosses/HeavyLamp/GE_ClearAbilities_HeavyLamp.GE_ClearAbilities_HeavyLamp_C" 0.0 _returnvalue_2)
  (GameplayEffects|ApplyGameplayEffectToSelf _returnvalue "/JethroLager/NPCs/Bosses/HeavyLamp/GE_HandleWeaponDBNO_HeavyLamp.GE_HandleWeaponDBNO_HeavyLamp_C" 0.0 _returnvalue_2)
  (bind _asfort_athena_aibot_controller (Utilities|Casting|CastToFortAthenaAIBotController (Fortnite|GetControllerfromActor self)))
  (Utilities|IsValid (Actor|GetComponentbyClass _asfort_athena_aibot_controller "/Script/FortniteAI.FortAthenaNpcPatrollingComponent")
    (:Is%20Valid
      (Class|FortAthenaNpcPatrollingComponent|RemovePatrolPath (Actor|GetComponentbyClass _asfort_athena_aibot_controller "/Script/FortniteAI.FortAthenaNpcPatrollingComponent"))
      (Class|FortAthenaAIBotController|SetRadialLeashLocation _asfort_athena_aibot_controller (Transformation|GetActorLocation))
      (Class|FortAthenaAIBotController|SetRadialLeashInnerRadius _asfort_athena_aibot_controller 2.5e+03)
      (Class|FortAthenaAIBotController|SetRadialLeashOuterRadius _asfort_athena_aibot_controller 3e+03)
      (if (not (Variables|Conversion|GetIsConverted))
        (CallFunction|SetCurrentMovementStyle "Walking")))))

(fn HandlePawnConverted
  (params (InstigatorPawn :type Fort%20Pawn%20Object%20Reference) (ConvertedPawn :type Fort%20Pawn%20Object%20Reference))
  (|CallParentFunction InstigatorPawn ConvertedPawn)
  (Variables|Conversion|SetIsConverted true)
  (Variables|Conversion|SetLeader%20Team (Team|GetActorTeam InstigatorPawn))
  (Conversion|UpdateConvertIndicator)
  (Conversion|GetSquadPawns)
  (Conversion|SetSquadsTeamFactioning true)
  (CallFunction|AddJethroControllerComponenttoBossTeam)
  (Conversion|HealSquadMembersonHire))

(fn HandlePawnUnconverted
  (params (UnconvertedPawn :type Fort%20Pawn%20Object%20Reference))
  (|CallParentFunction UnconvertedPawn)
  (Variables|Conversion|SetIsConverted false)
  (Variables|Conversion|SetLeader%20Team 0)
  (Conversion|UpdateConvertIndicator)
  (Conversion|SetSquadsTeamFactioning false))

(fn HealSquadMembersOnHire
  (bind _convertcomponent (Variables|FortPlayerPawn|GetConvertComponent))
  (bind _returnvalue (Convert|GetConverterPawn _convertcomponent))
  (Networking|SwitchHasAuthority
    (:Authority
      (if (not (Variables|Default|GetHasAlreadyHealedOnHire))
        (Variables|Default|SetHasAlreadyHealedOnHire true)
        (if (ScalableFloat|GetValueasBool (Variables|Conversion|GetShouldHealSquadOrConverterOnHire))
          (if (ScalableFloat|GetValueasBool (Variables|Conversion|GetOnlyHealConverterOnHire))
            (Utilities|IsValid _returnvalue
              (:Is%20Valid
                (Environment|ApplyGlobalEnvironmentGameplayEffecttoActor _returnvalue (Variables|Default|GetGE_HealSquadOnConvert))))
            (else
              (for _array_element (Variables|Conversion|GetConvertedSquadPawns)
                (Environment|ApplyGlobalEnvironmentGameplayEffecttoActor _array_element (Variables|Default|GetGE_HealSquadOnConvert))))))))))

(fn HeavyLamp_OnDamaged
  (params (DamagedActor :type Actor%20Object%20Reference) (Damage :type Float%20%28single-precision%29) (InstigatedBy :type Controller%20Object%20Reference) (DamageCauser :type Actor%20Object%20Reference) (HitLocation :type Vector) (FHitComponent :type Primitive%20Component%20Object%20Reference) (BoneName :type Name) (Momentum :type Vector))
  (bind _asfort_player_controller (Utilities|Casting|CastToFortPlayerController InstigatedBy))
  (Conversion|AssignLastDamagingPawn (Game|Player|GetPlayerPawn _asfort_player_controller)))

(fn IsValidAutoFireTarget
  (params (Attacker :type Actor%20Object%20Reference))
  (return (not (Variables|RiftTransition|GetBossFightFinished))))

(fn JumpToPathLocation
  (bind _path_points (Variables|RiftTransition|GetPath%20Points))
  (if (Variables|Default|GetWasForceTeleported)
    _
    (elif (Utilities|Array|IsValidIndex _path_points)
      (RiftTransition|MakePathPointValid (Utilities|Array|Get(acopy) _path_points))
      (else
        (RiftGate|SetupConversationBGA false)))))

(fn MakePathPointValid
  (params (PathPoint :type Vector))
  (bind _distance (CallFunction|ActorDistancetoPoint PathPoint))
  (bind _self self)
  (if (< _distance (ScalableFloat|GetValueatLevel (Variables|Movement|GetMaxDistanceToAllowLeap)))
    (bind _output_get (Variables|RiftTransition|SetRiftGateJumpLocation PathPoint))
    (if (Utilities|Operators|Equal(==) (Utilities|Array|Get(acopy) (Variables|RiftTransition|GetPath%20Points)) _output_get 1.0)
      (Utilities|Array|RemoveIndex (Variables|RiftTransition|GetPath%20Points))
      (else
        _))
    (if (> _distance (ScalableFloat|GetValueatLevel (Variables|Movement|GetMinDistanceToAllowLeap)))
      (GameplayEffects|ApplyGameplayEffectToSelf (Ability|GetFortAbilitySystemComponent _self) (Variables|RiftTransition|GetHeavyLampLeapGE) 0.0 (Class|AthenaFunctionLibrary|MakeSimpleFortEffectContext _self))
      (else
        (bind _returnvalue (Game|SpawnActorfromClass "/JethroLager/SmartObjects/HeavyLamp/SmartObject_NoMesh_Actor_HeavyLamp.SmartObject_NoMesh_Actor_HeavyLamp_C" (Math|Transform|MakeTransform (Variables|RiftTransition|GetRiftGateJumpLocation)) "Undefined" "MultiplyWithRoot" 0 _self))
        (bind _pathsmartobject (Variables|RiftTransition|GetPathSmartObject)
          (:then
            (Actor|DestroyActor _pathsmartobject)
            (Variables|RiftTransition|SetPathSmartObject _returnvalue))
          (:else))))
    (else
      (bind _output_get_2 (Variables|RiftTransition|SetRiftGateJumpLocation (CallFunction|DivideVector 2.0 PathPoint)))
      (RiftTransition|MakePathPointValid _output_get_2))))

(fn OffsetNPCLocationFromRiftGate
  (bind _hfrow_npcriftgateoffsetright (Variables|Default|GetHFRow_NPCRiftGateOffsetRight))
  (bind _returnvalue (ScalableFloat|GetValueatLevel _hfrow_npcriftgateoffsetright))
  (bind _hfrow_npcriftgateoffsetforward (Variables|RiftGate|GetHFRow_NPCRiftGateOffsetForward))
  (bind _returnvalue_2 (ScalableFloat|GetValueatLevel _hfrow_npcriftgateoffsetforward))
  (bind _hfrow_npcriftgateoffsetright_2 (Variables|Default|GetHFRow_NPCRiftGateOffsetRight))
  (bind _returnvalue_3 (ScalableFloat|GetValueatLevel _hfrow_npcriftgateoffsetright_2))
  (bind _hfrow_npcriftgateoffsetforward_2 (Variables|RiftGate|GetHFRow_NPCRiftGateOffsetForward))
  (bind _returnvalue_4 (ScalableFloat|GetValueatLevel _hfrow_npcriftgateoffsetforward_2))
  (bind _lastdamagerplayercontroller (Variables|Default|GetLastDamagerPlayerController)
    (:then
      (Utilities|IsValid (Pawn|GetControlledPawn _lastdamagerplayercontroller)
        (:Is%20Valid
          (LocalVariables|Seth_CompareAgainstSourceLocation (Transformation|GetActorLocation (Pawn|GetControlledPawn _lastdamagerplayercontroller))))
        (:Is%20Not%20Valid
          (LocalVariables|Seth_CompareAgainstSourceLocation (Transformation|GetActorLocation))))
      (bind _spawnedriftgate (Variables|RiftGate|GetSpawnedRiftGate)
        (:then
          (if (> (Math|Vector|Distance2DSquared(Vector) (Utilities|Operators|Add (Transformation|GetActorLocation _spawnedriftgate) (* (Transformation|GetActorForwardVector _spawnedriftgate) _returnvalue_2) (* (Transformation|GetActorRightVector _spawnedriftgate) _returnvalue_3)) (LocalVariables|Geth_CompareAgainstSourceLocation)) (Math|Vector|Distance2DSquared(Vector) (LocalVariables|Geth_CompareAgainstSourceLocation) (Utilities|Operators|Add (Transformation|GetActorLocation _spawnedriftgate) (* (Transformation|GetActorForwardVector _spawnedriftgate) _returnvalue_4) (Utilities|Operators|Multiply (Transformation|GetActorRightVector _spawnedriftgate) _returnvalue -1.0))))
            ;[0.309 0.4767 0.763] Check RIGHT side offset
            (return (Utilities|Operators|Add (Transformation|GetActorLocation _spawnedriftgate) (* (Transformation|GetActorForwardVector _spawnedriftgate) _returnvalue_2) (* (Transformation|GetActorRightVector _spawnedriftgate) _returnvalue_3)))
            (else
              ;[0.309 0.4767 0.763] Check LEFT side offset
              (return (Utilities|Operators|Add (Transformation|GetActorLocation _spawnedriftgate) (* (Transformation|GetActorForwardVector _spawnedriftgate) _returnvalue_4) (Utilities|Operators|Multiply (Transformation|GetActorRightVector _spawnedriftgate) _returnvalue -1.0))))))
        (:else
          ;[1 0.49 0.7753] RETURN No valid offset target
          (return (Transformation|GetActorLocation)))))
    (:else
      _)))

(fn OnRep_AITarget
  (Networking|SwitchHasAuthority
    (:Remote
      (Utilities|IsValid (Variables|Default|GetAITarget)
        (:Is%20Valid
          (CallFunction|StartYawUpdateTimer))
        (:Is%20Not%20Valid
          (Utilities|Time|ClearandInvalidateTimerbyHandle (Variables|Default|GetHNDL_UpdateTargettingYaw))
          (Variables|Animation|SetTargetYaw 0.0))))))

(fn OnRep_BossFightFinished
  (Networking|CanExecuteCosmeticEvents
    (:True
      (Default|CallOnBossFightFinished (Variables|RiftTransition|GetBossFightFinished))
      (MusicEncounter|ExecuteMusicEncounterGameplayEventByName (GameplayEventRouter|GetGlobalEventRouter) "Defeated" self))))

(fn OnRep_CanEnableHealthBars
  (if (Variables|Default|GetCanEnableHealthBars)
    (CallFunction|NPCEnableHealthBars)))

(fn OnRep_IsConverted
  (Networking|CanExecuteCosmeticEvents
    (:True
      (Default|CallOnHeavyLampConverted (Variables|Conversion|GetIsConverted)))))

(fn OnRep_IsCurrentlyDeflecting
  (Networking|SwitchHasAuthority))

(fn OnRep_SetInvisible
  (Networking|SwitchHasAuthority
    (:Remote
      (FortPawn|SetPawnVisibilityLocal (Variables|Default|GetSetInvisible)))))

(fn OnRep_ShowIndicatorMarker
  (Networking|SwitchHasAuthority
    (:Remote
      (if (Variables|Conversion|GetIsConverted)
        (if (Variables|Default|GetShowIndicatorMarker)
          (if (Math|Byte|Equal(Byte) (Team|GetActorTeam (Fortnite|GetFirstLocalFortPlayerController true)) (Variables|Conversion|GetLeader%20Team))
            (Convert|ShowIndicatorMarker (Variables|FortPlayerPawn|GetConvertComponent))
            (else
              (Convert|HideIndicatorMarker (Variables|FortPlayerPawn|GetConvertComponent))))
          (else
            (Convert|HideIndicatorMarker (Variables|FortPlayerPawn|GetConvertComponent))))))))

(fn OnRep_TargetYaw)

(fn OnRep_UpdateToQueryWorldPosition
  (Collision|IgnoreActorwhenMoving (Variables|Character|GetCapsuleComponent) (Variables|RiftGate|GetSpawnedRiftGate) true)
  (RiftGate|FindRiftPath))

(fn OnRep_WeaponHolsteredOnDeath)

(fn PlayKneelAnim
  (params (KneelDuration :type Float%20%28double-precision%29))
  (bind _returnvalue (Animation|PlayAnimMontage (Variables|RiftTransition|GetAnim_TakeKnee)))
  (return (+ KneelDuration _returnvalue)))

(fn PlayStandupAnim
  (bind _anim_standfromknee (Variables|RiftTransition|GetAnim_StandFromKnee)
    (:then
      (bind _returnvalue (Animation|PlayAnimMontage _anim_standfromknee))
      (return _returnvalue))
    (:else
      (bind _returnvalue_2 (Animation|PlayAnimMontage (Variables|RiftTransition|GetAnim_TakeKnee) -1.0))
      (return _returnvalue_2))))

(fn SetSpawnerAfterEQSQueryComplete
  (params (QueryInstance :type EQS%20Query%20Instance%20Object%20Reference) (QueryStatus :type EEnvQueryStatus%20Enum :default "Processing"))
  (AI|EQS|UnbindEventfromOnQueryFinished QueryInstance (EventDispatchers|CreateEvent :!event SetSpawnerAfterEQSQueryComplete))
  (bind (_besteqslocation _resultfound) (RiftGate|GetBestRiftGateLocationfromEQSResults QueryInstance))
  (if _resultfound
    (RiftGate|SpawnHeavyLampRift _besteqslocation)
    (bind _bestoffsetlocation (RiftGate|OffsetNPCLocationfromRiftGate))
    (Variables|RiftGate|SetUpdateToQueryWorldPosition _bestoffsetlocation)
    (bind _returnvalue (Utilities|Time|SetTimerbyEvent (EventDispatchers|CreateEvent :!event FallbackForceTeleport) (ScalableFloat|GetValueatLevel (Variables|Default|GetForceTeleportFallbackDelay))))
    (Variables|Default|SetFallbackForceTeleportTimerHandle _returnvalue)
    (else
      (Development|PrintString "RIFTGATEEQS: No Valid result" true true "(R=1.000000,G=0.000000,B=0.270694,A=1.000000)" 12.0)
      (if (< (Variables|RiftGate|GetRiftPathReadyRetries) (Variables|RiftGate|GetMaxRiftPathReadyRetries))
        (Math|Integer|IncrementInt (Variables|RiftGate|GetRiftPathReadyRetries)
          (:%20%20
            (Utilities|Time|SetTimerbyEvent (EventDispatchers|CreateEvent :!event CheckLocationForRiftGateSpawn) (ScalableFloat|GetValueatLevel (Variables|Default|GetEQSRetryTimer)))))
        (elif (ScalableFloat|GetValueasBool (Variables|Default|GetDieOnRiftFailed))
          (CallFunction|StartHeavyLampDespawn))))))

(fn SetSquadsTeamFactioning
  (params (IsApplying :type Boolean))
  (bind _lonefeatherfaction (Variables|Conversion|GetLoneFeatherFaction))
  (if (Utilities|Array|IsNotEmpty (Variables|Conversion|GetConvertedSquadPawns))
    (for _array_element (Variables|Conversion|GetConvertedSquadPawns)
      (if IsApplying
        (Affiliation|AddtoFaction _array_element _lonefeatherfaction)
        (if (ScalableFloat|GetValueasBool (Variables|Default|GetCanSetNeutral))
          _
          (elif (ScalableFloat|GetValueasBool (Variables|Default|GetCanSetFriendly))
            (Variables|Default|SetFactionAttitude "Friendly")))
        (Affiliation|ChangeFactionAttitudeTowardsActor (GameplayTags|MakeGameplayTagContainerfromTag (Variables|Conversion|GetLoneFeatherFaction)) _array_element (Variables|Default|GetFactionAttitude))
        (else
          (Affiliation|RemovefromFaction _array_element _lonefeatherfaction))))))

(fn SetupConversationBGA
  (params (UseTeleportLocation :type Boolean))
  (bind _spawnedconversationactor (Variables|Default|GetSpawnedConversationActor)
    (:then
      (Class|BGAAthenaPlotTwistHeavyLampConvo|ForceRemove _spawnedconversationactor))
    (:else
      _))
  (bind _returnvalue (Game|SpawnBuildingActorfromClass "/PlotTwist/Items/RiftGate/RiftGate_HeavyLamp/BGA_Athena_PlotTwist_HeavyLamp_Convo.BGA_Athena_PlotTwist_HeavyLamp_Convo_C" (Math|Transform|MakeTransform (Math|Vector|SelectVector (Variables|Default|GetForceTeleportLocation) (Transformation|GetActorLocation self) UseTeleportLocation)) 0 false 0 self))
  (bind _output_get (Variables|Default|SetSpawnedConversationActor _returnvalue))
  (Variables|/PlotTwist/Items/RiftGate/RiftGate_HeavyLamp/BGA_Athena_PlotTwist_HeavyLamp_Convo.BGA_Athena_PlotTwist_HeavyLamp_Convo_C|SetOwningNPCPawn self _output_get)
  (bind _spawnedriftgate (Variables|RiftGate|GetSpawnedRiftGate)
    (:then
      (Variables|/PlotTwist/Items/RiftGate/RiftGate_HeavyLamp/BGA_Athena_PlotTwist_HeavyLamp_Convo.BGA_Athena_PlotTwist_HeavyLamp_Convo_C|SetAssociatedRiftGate _spawnedriftgate _output_get) ; since the plugin dependency is \nAI -> PlotTwist, it is difficult for \nthose actors to get this boss and\neach other
    )
    (:else
      _))
  (Default|BindEventtoTrainingAcceptedbyConversation _output_get (EventDispatchers|CreateEvent :!event BeginRiftTrainingSequence))
  (Default|BindEventtoForcedEndOfTraining _output_get (EventDispatchers|CreateEvent :!event ExitWithoutTraining)))

(fn SpawnHeavyLampRift
  (params (SpawnLocation :type Vector))
  (bind (_returnvalue_roll _returnvalue_pitch _returnvalue_yaw) (Math|Rotator|BreakRotator (Math|Rotator|FindLookatRotation SpawnLocation (Transformation|GetActorLocation self))))
  (bind _heavylamp_specialactor (Variables|Default|GetHeavyLamp_SpecialActor))
  (Variables|RiftGate|GetSpawnedRiftGate
    (:then)
    (:else
      (Game|AddComponentbyClass "/Game/Athena/AI/NPCs/Tandem/Base/Components/NPCLoopStateComponent.NPCLoopStateComponent_C")
      (bind _returnvalue (Game|SpawnBuildingActorfromClass "/PlotTwist/Items/RiftGate/RiftGate_HeavyLamp/B_Athena_PlotTwist_RiftGate_HeavyLamp_OpenRift.B_Athena_PlotTwist_RiftGate_HeavyLamp_OpenRift_C" (Math|Transform|MakeTransform SpawnLocation (Math|Rotator|MakeRotator 0.0 0.0 _returnvalue_yaw)) 0 false 0 0 self))
      (bind _output_get (Variables|RiftGate|SetSpawnedRiftGate _returnvalue))
      (Variables|/PlotTwist/Items/RiftGate/RiftGate_HeavyLamp/B_Athena_PlotTwist_RiftGate_HeavyLamp_OpenRift.B_Athena_PlotTwist_RiftGate_HeavyLamp_OpenRift_C|SetOwningNPCPawn self _output_get)
      ;[0.1954 0 0.4661] Disable Map Icon
      (Components|Activation|Deactivate _heavylamp_specialactor)
      (Class|AthenaSpecialActorComponent|SetMiniMapIcon _heavylamp_specialactor (Utilities|Struct|MakeSlateBrush "(SpecifiedColor=(R=1.000000,G=1.000000,B=1.000000,A=1.000000),ColorUseRule=UseColor_Specified)"))
      (Utilities|Time|ClearandInvalidateTimerbyHandle (Variables|Default|GetDisplayMapMarkerTimerHandle)))))

(fn StartPathToRift
  (bind _path_points (Variables|RiftTransition|GetPath%20Points))
  (bind _output (Utilities|Array|Get(acopy) _path_points))
  (if (Utilities|Array|IsValidIndex (Variables|RiftTransition|GetPath%20Points))
    (if (<= (CallFunction|ActorDistancetoPoint _output) (+ (ScalableFloat|GetValueatLevel (Variables|Movement|GetMinDistanceToAllowLeap)) 2e+02))
      (Variables|RiftTransition|SetRiftGateJumpLocation _output)
      (Utilities|Array|Clear (Variables|RiftTransition|GetPath%20Points))
      (else
        ;[1 1 1 0.5] Location is greater than min jump distance so we want to only walk a short distance
        (Variables|RiftTransition|SetRiftGateJumpLocation (CallFunction|DivideVector 5.0 _output))))
    (bind _returnvalue (Game|SpawnActorfromClass "/JethroLager/SmartObjects/HeavyLamp/SmartObject_NoMesh_Actor_HeavyLamp.SmartObject_NoMesh_Actor_HeavyLamp_C" (Math|Transform|MakeTransform (Variables|RiftTransition|GetRiftGateJumpLocation)) "Undefined" "MultiplyWithRoot" 0 self))
    (Variables|RiftTransition|SetPathSmartObject _returnvalue)
    (elif (ScalableFloat|GetValueasBool (Variables|Default|GetCanRetryRiftPath))
      ;[0.9323 0.1309 0.04172 0.5] No valid points we need to find new path
      (RiftGate|FindRiftPath))))

(fn UpdateConvertIndicator
  (bind _convertcomponent (Variables|FortPlayerPawn|GetConvertComponent))
  (Networking|SwitchHasAuthority
    (:Remote
      (if (Math|Byte|Equal(Byte) (Team|GetActorTeam (Fortnite|GetFirstLocalFortPlayerController true)) (Variables|Conversion|GetLeader%20Team))
        (Convert|ShowIndicatorMarker _convertcomponent)
        (Default|BindEventtoOnDeath (EventDispatchers|CreateEvent :!event Play%202D%20Death%20VO%20for%20Team))
        (else
          (Convert|HideIndicatorMarker _convertcomponent))))))

(fn ConstructionScript
  (|CallParentFunction))
