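; HeavyLamp spawn manager that chooses where the Vader boss and its shuttle spawn by activating Data Layers.
; On authority, it waits for playlist data, then checks a hotfixable flag for random spawning.
; After validating that the location and shuttle arrays match in length, it activates a random same-index pair and logs it.
; Default location and shuttle layers cover the disabled or invalid case, with shipping logs explaining why.
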(event EventBeginPlay
  (Networking|SwitchHasAuthority
    (:Authority
      (Class|FortAsyncActionCurrentPlaylistReady|PlaylistDataReadyAsync
        (:OnReady
          (if (ScalableFloat|GetValueasBool (Variables|Default|GetHF_SpawnAtRandomLocation))
            (bind _isvalid (CallFunction|ValidateData))
            (if _isvalid
              (bind _output_get (Variables|InternalData|SetPickedRandomIndex (Math|Random|RandomIntegerinRange 0 (- (Utilities|Array|Length (Variables|Default|GetRandomLocationDataLayerNames)) 1))))
              (bind _output_get_2 (Variables|InternalData|SetPickedRandomDataLayer (Utilities|Array|Get(aref) (Variables|Default|GetRandomLocationDataLayerNames) _output_get)))
              (bind _output_get_3 (Variables|InternalData|SetPickedShuttleDataLayer (Utilities|Array|Get(aref) (Variables|Default|GetShuttleDataLayerNames) _output_get)))
              (Fortnite|FortShippingLog self (Utilities|String|ToString(Text) (Utilities|Text|FormatText "NSLOCTEXT(\"[95E5F6C849EEDA4E10C6DCB2BC0E1AEF]\", \"D855D696456445FDA6DC278DCD552B24\", \"HeavyLamp Manager picked the following DataLayer to load: {a} for the Random Spawn Location of Vader. The Shuttle Layer that will also load is {b}.\")" _output_get_2 _output_get_3)) true)
              (DataLayers|SetDataLayerRuntimeStatebyLabel (WorldSubsystems|GetDataLayerSubsystem) (Variables|InternalData|GetPickedRandomDataLayer) "Activated" false)
              (DataLayers|SetDataLayerRuntimeStatebyLabel (WorldSubsystems|GetDataLayerSubsystem) (Variables|InternalData|GetPickedShuttleDataLayer) "Activated" false)
              (else
                _))
            (else
              ; Load Default Spawn Location for Vader
              (Fortnite|FortShippingLog self "HeavyLamp Random Spawn Location Disabled. Spawning at the Default Location." true)))
          (DataLayers|SetDataLayerRuntimeStatebyLabel (WorldSubsystems|GetDataLayerSubsystem) (Variables|Default|GetDefaultDataLayerName) "Activated" false)
          (DataLayers|SetDataLayerRuntimeStatebyLabel (WorldSubsystems|GetDataLayerSubsystem) (Variables|Default|GetDefaultShuttleDataLayerName) "Activated" false))))))

(fn ConstructionScript)

(fn ValidateData
  (bind _randomlocationdatalayernames (Variables|Default|GetRandomLocationDataLayerNames))
  (bind _returnvalue (Utilities|Array|Length _randomlocationdatalayernames))
  (if (> _returnvalue 0)
    (if (== _returnvalue (Utilities|Array|Length (Variables|Default|GetShuttleDataLayerNames)))
      (return true)
      (else
        (Fortnite|FortShippingLog self "Shuttle Location Array Count does not match Random Location Array Count! Make sure you have the same amount of entries in both arrways to use this feature!" true)))
    (else
      (Fortnite|FortShippingLog self "HF mentions to Load a Random Location, but no Random Locations were defined. Loading Default DataLayer" true)))
  (return false))
