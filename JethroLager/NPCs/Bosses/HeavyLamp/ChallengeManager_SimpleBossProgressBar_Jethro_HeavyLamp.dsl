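; Challenge manager for the Heavy Lamp boss fight, driving a simple boss progress bar and completing the challenge when the boss is downed.
; On spawn it starts the challenge, leashes the boss and binds damage, DBNO and a failsafe death event.
; First damage resizes the event area collision; DBNO completes the challenge and spawns the reward.
; Bar progress blends health and shield percentages, weighted by max-value scalars indexed by playlist team size.
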(event EventBeginPlay
  (|CallParentFunction)
  (bind (_gamestate _playlist) (Class|FortAsyncActionCurrentPlaylistReady|PlaylistDataReadyAsync)
    (:OnReady
      (AI|SetLagerDatafromPlaylistSize _gamestate _playlist))))

(event Custom|OnHeavyLampDamaged
  (params (DamagedActor :type Actor%20Object%20Reference) (Damage :type Float%20%28single-precision%29) (InstigatedBy :type Controller%20Object%20Reference) (DamageCauser :type Actor%20Object%20Reference) (HitLocation :type Vector) (FHitComponent :type Primitive%20Component%20Object%20Reference) (BoneName :type Name) (Momentum :type Vector))
  (EventDispatchers|UnbindEventfromOnDamaged (Variables|Default|GetSpawnedPawn) (delegate Custom|OnHeavyLampDamaged))
  (Transformation|SetRelativeTransform (Variables|Default|GetEventAreaCollision) (struct :Location "195.000000,0.000000,0.000000" :Rotation "0, 0, 0" :Scale "1.800000,1.500000,1.000000") true false))

(event Collision|EventActorBeginOverlap
  (params (OtherActor :type Actor%20Object%20Reference))
  (|CallParentFunction OtherActor) ; This node is disabled and will not be called.\nDrag off pins to build functionality.
)

(event EventTick
  (params (DeltaSeconds :type Float%20%28single-precision%29))
  (|CallParentFunction DeltaSeconds) ; This node is disabled and will not be called.\nDrag off pins to build functionality.
)

(event EventAISpawned
  (params (SpawnedActor :type Actor%20Object%20Reference) (bResult :type Boolean))
  (CallFunction|HandleBossSpawned SpawnedActor)
  (CallFunction|SetIsBossReadyforFinalFight true)
  (CallFunction|StartSimpleChallenge)
  (Leash|UpdateAIActorLeash SpawnedActor)
  (Music|AddCurrentMusicTagtoActor SpawnedActor)
  (bind _asfort_player_pawn_athena (Utilities|Casting|CastToFortPlayerPawnAthena SpawnedActor))
  (bind _output_get (Variables|Default|SetSpawnedPawn _asfort_player_pawn_athena))
  (EventDispatchers|BindEventtoOnDamaged _output_get (EventDispatchers|CreateEvent :!event Custom|OnHeavyLampDamaged))
  (EventDispatchers|BindEventtoOnPlayerStartDBNO _output_get (delegate Custom|OnHeavyLampDBNO))
  ;[1 0 0.7437] Bind/Call parent functionality for OnDied - this boss is expected to go into DBNO, but this is a failsafe in case it outright dies somehow (FORT-888887)
  ;[FORT-882105 is tracking the bug where this boss dies unexpectedly]
  (bind _asfort_athena_aibot_controller (Utilities|Casting|CastToFortAthenaAIBotController (Pawn|GetController _asfort_player_pawn_athena)))
  (Death|BindEventtoOnAgentDiedEvent _asfort_athena_aibot_controller (EventDispatchers|CreateEvent :!event AIDied)))

(event Custom|OnHeavyLampDBNO
  (Variables|Progress|SetTotalProgress 1)
  (CallFunction|UpdateChallengeState "Completed")
  (CallFunction|SpawnReward (Variables|Default|GetSpawnedPawn))
  (CallFunction|HandleBossDied))

(event EventHandleBossHealthChanged
  (bind _bossactor (Variables|Boss|GetBossActor))
  (if (Variables|Default|GetBossIsReadyForFinalFight)
    (Utilities|IsValid _bossactor
      (:Is%20Valid
        (bind _asbp_player_pawn_heavy_lamp (Utilities|Casting|CastToBP_PlayerPawn_HeavyLamp _bossactor))
        (if (not (Variables|/JethroLager/NPCs/Bosses/HeavyLamp/BP_PlayerPawn_HeavyLamp.BP_PlayerPawn_HeavyLamp_C|GetBossFightFinished _asbp_player_pawn_heavy_lamp))
          (Variables|Progress|SetBossProgress (Utilities|NormalizeFloat (+ (* (FortPawn|HealthPercent _asbp_player_pawn_heavy_lamp) (ScalableFloat|GetValueatLevel (Variables|Default|GetMaxHealthScalar) (Math|Conversions|ToFloat(Integer) (- (Variables|Default|GetPlaylistMaxTeamSize) 1)))) (* (FortPawn|ShieldPercent _asbp_player_pawn_heavy_lamp) (ScalableFloat|GetValueatLevel (Variables|Default|GetMaxShieldScalar) (Math|Conversions|ToFloat(Integer) (- (Variables|Default|GetPlaylistMaxTeamSize) 1))))) 0 (+ (ScalableFloat|GetValueatLevel (Variables|Default|GetMaxHealthScalar) (Math|Conversions|ToFloat(Integer) (- (Variables|Default|GetPlaylistMaxTeamSize) 1))) (ScalableFloat|GetValueatLevel (Variables|Default|GetMaxShieldScalar) (Math|Conversions|ToFloat(Integer) (- (Variables|Default|GetPlaylistMaxTeamSize) 1)))) true)))))))

(fn ConstructionScript
  (|CallParentFunction))
