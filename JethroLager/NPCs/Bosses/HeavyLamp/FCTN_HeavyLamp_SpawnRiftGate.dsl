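; Conversation node for the HeavyLamp boss NPC that spawns its conversation Rift Gate at the talking player's pawn.
; The spawn happens as a side effect inside IsRequirementSatisfied, so evaluating the requirement creates the actor.
; It casts the conversation's player participant to an Athena controller, resolves the pawn, spawns at that transform and returns Passed.
; A failed cast or missing pawn returns FailedAndHidden.
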(fn IsRequirementSatisfied
  (params (Context :type Conversation%20Context%20Structure%20%28by%20ref%29))
  (bind _returnvalue (Class|FortConversationContextLibrary|GetParticipantActorPlayer Context))
  (bind _asfort_player_controller_athena (Utilities|Casting|CastToFortPlayerControllerAthena _returnvalue)
    (:then
      (bind _fortpawn (Class|FortQuestBlueprintLibrary|GetFortPawnFromObject _asfort_player_controller_athena)
        (:Valid
          (Game|SpawnActorfromClass "/PlotTwist/Items/RiftGate/RiftGate_HeavyLamp/B_Athena_PlotTwist_RiftGate_HeavyLamp_Convo.B_Athena_PlotTwist_RiftGate_HeavyLamp_Convo_C" (Transformation|GetActorTransform _fortpawn) "Undefined" "MultiplyWithRoot" 0 0 0 _fortpawn)
          (return "Passed"))
        (:NotValid
          _)))
    (:CastFailed
      _))
  (return "FailedAndHidden"))
