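; Gameplay ability for the HeavyLamp boss that, per its name, heals squad members on conversion by topping up the avatar pawn.
; On activation it gets the avatar as a valid Fort player pawn, then checks two scalable-float toggles.
; Each enabled toggle sets shield or health to its maximum and executes the matching gameplay cue on that pawn.
; The ability then ends.
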(event Ability|EventActivateAbility
  (bind _outfortplayerpawn (CallFunction|GetAvatarasValidFortPlayerPawn)
    (:ValidFortPlayerPawn
      (if (ScalableFloat|GetValueasBool (Variables|Default|GetShouldSetShields))
        (FortPawn|SetShield _outfortplayerpawn (FortPawn|GetMaxShield _outfortplayerpawn))
        (GameplayCue|ExecuteGameplayCue _outfortplayerpawn (Variables|Default|GetGC_Shield))
        (else
          _))
      (if (ScalableFloat|GetValueasBool (Variables|Default|GetShouldSetHealth))
        (FortPawn|SetHealth _outfortplayerpawn (FortPawn|GetMaxHealth _outfortplayerpawn))
        (GameplayCue|ExecuteGameplayCue _outfortplayerpawn (Variables|Default|GetGC_Heal))
        (else
          _)))
    (:AvatarCastFailed
      _))
  (Ability|EndAbility))

(event Ability|EventOnEndAbility
  (params (bWasCancelled :type Boolean)))
