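; Gameplay ability for the HeavyLamp boss that refills its health, intended by its name to run once the boss is converted.
; On activation it validates the avatar as a Fort pawn, waits one tick, then applies the configured gameplay effect to the owner.
; The pawn's max health is passed as the SetByCaller.Healing magnitude, so the effect can heal fully.
; The ability then ends, or ends immediately if the avatar cast fails.
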(event Ability|EventActivateAbility
  (bind _outfortplayerpawn (CallFunction|GetAvatarasValidFortPlayerPawn)
    (:ValidFortPlayerPawn
      (Utilities|FlowControl|DelayUntilNextTick)
      (bind _returnvalue (Ability|GameplayEffect|AssignTagSetbyCallerMagnitude (Ability|MakeOutgoingGameplayEffectSpec (Variables|Default|GetGE_EffectToApply)) "(TagName=\"SetByCaller.Healing\")" (FortPawn|GetMaxHealth _outfortplayerpawn)))
      (Ability|ApplyGameplayEffectSpecToOwner _returnvalue)
      (Ability|EndAbility))
    (:AvatarCastFailed
      (Ability|EndAbility))))

(event Ability|EventOnEndAbility
  (params (bWasCancelled :type Boolean)))

(event Ability|EventFailedToActivatePassiveAbility
  (params (ActorInfo :type Gameplay%20Ability%20Actor%20Info%20Structure)))
