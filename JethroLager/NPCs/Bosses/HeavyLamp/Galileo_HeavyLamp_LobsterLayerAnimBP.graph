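# Item animation layer for the HeavyLamp boss pawn on the player skeleton, supplying deflect, aim and airborne poses.
# An upper-body state machine plays a deflect loop, then flag-selected exit animations using inertialization.
# Update sets aim-offset yaw toward nearby living player targets, though a later Sequence branch zeroes it, and eases a surface-swimming alpha that lowers the root.
# Melee guarding switches jump, fall, jetpack and zipline additives; a curve gates root-rotation correction.
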AnimBlueprint /JethroLager/NPCs/Bosses/HeavyLamp/Galileo_HeavyLamp_LobsterLayerAnimBP
target_skeleton=&Fortnite_M_Avg_Player_Skeleton
graphs [
 AnimGraph {nodes [Output Pose() {layerGroup=None}]}
 EventGraph {nodes [
   n1 = Event BlueprintUpdateAnimation()
   > Set Delta Time X(Delta Time X=@n1.DeltaTimeX)
   > n2 = Sequence()
   n3 = Property Access!() -> (Value:Fort Pawn Object Reference) {is=title K2Node_PropertyAccess}
   > n4 = Cast To BP_PlayerPawn_HeavyLamp!(execute, Object=@n3) -> (then, CastFailed, AsBP Player Pawn Heavy Lamp:BP Player Pawn Heavy Lamp Object Reference) {is=title K2Node_DynamicCast}
   n5 = Get AITarget!(self:BP Player Pawn Heavy Lamp Object Reference=@n4) -> (AITarget:Actor Object Reference)
   > n6 = Cast To FortPlayerPawn(Object=@n5)
   n7 = IsDead!(self:Fort Pawn Object Reference=@n6) -> (ReturnValue:bool)
   n8 = "Distance (Vector)"!(V1:Vector=Get Actor Location(self=@n5), V2:Vector=Get Actor Location(self=@n3)) -> (ReturnValue:double)
   > Branch(Condition=AND Boolean(A=NOT Boolean(A=@n7), B="float <= float"(A=@n8, B=8000) -> (ReturnValue=~)))
   > Is Valid(InputObject=@n5)
   > Set Temp_Replacement_Yaw(Temp_Replacement_Yaw:double=Get TargetYaw() -> (TargetYaw:float)) -> (Output_Get:double)
   Set Temp_Replacement_Yaw(execute=@n2.then_1, Temp_Replacement_Yaw=0)
   n9 = IsSurfaceSwimming!(self:Fort Pawn Object Reference=Property Access()) -> (ReturnValue:bool)
   > n10 = Branch(Condition=@n9)
   > Set SurfaceSwimmingOffsetAlpha(SurfaceSwimmingOffsetAlpha:double=FInterpTo(
     Current=Get SurfaceSwimmingOffsetAlpha() -> (SurfaceSwimmingOffsetAlpha:double)
     Target="To Float (Boolean)"(InBool=@n9)
     DeltaTime=Get Delta Time X()
     InterpSpeed=3
    )) -> (Output_Get:double)
   Set SurfaceSwimmingOffsetAlpha(execute=@n10.else, SurfaceSwimmingOffsetAlpha=0)
  ]}
 ItemAimOffset {nodes [
   Input Pose!() -> (Pose, AimOffsetAlpha:double, Yaw:double, Pitch:double) {name=InPose_AimOffset}
   Input Pose!() -> (Pose) {name=InPose_PreAimOffsets}
   Output Pose() {layerGroup=Item Override}
  ]}
 ItemFallAdditive {nodes [
   Input Pose() {name=InPoseItemFallAdditive}
   n1 = Get FallPlayRate() -> (FallPlayRate:float)
   Output Pose(Result=Blend Poses by bool(
     BlendPose_0="Sequence Player (additive) 'M_JumpFall_ARDS_Additive'"!(PlayRate:float=@n1) -> (Pose) {bLoopAnimation=false, sequence=/Game/Animation/Game/MainPlayer/Locomotion/Medium/Male/Jump/AssaultRifle/M_JumpFall_ARDS_Additive}
     BlendPose_1="Sequence Player (additive) 'JumpFall_NoWep_CMM_Additive'"!(PlayRate:float=@n1) -> (Pose) {bLoopAnimation=false, sequence=/Game/Animation/Game/MainPlayer/Locomotion/Medium/Male/Jump/NoWep/JumpFall_NoWep_CMM_Additive}
     bActiveValue=Get bIsMeleeGuarding() -> (bIsMeleeGuarding:bool)
    ) {blendType=Linear}) {layerGroup=Item Override}
  ]}
 ItemFallLandAdditive {nodes [
   Input Pose() {name=InPoseItemFallLandAdditive}
   Output Pose() {layerGroup=Item Override}
  ]}
 ItemFallLevel2Override {nodes [
   Input Pose() {name=InPose_FallLevel2}
   Output Pose() {layerGroup=Item Override}
  ]}
 ItemFinalPoseOverride {nodes [
   n1 = Input Pose() {name=InFinalPose}
   n2 = Get Temp_Replacement_Yaw() -> (Temp_Replacement_Yaw:double)
   n3 = "Transform (Modify) Bone - Bone: Root"(
    ComponentPose=Local To Component(LocalPose=AimOffset Player 'M_NoWep_Relaxed_AO'!(X:float=@n2, Y:float=0, BasePose:Pose Link Structure=@n1, Alpha:float=1) -> (Pose:Pose Link Structure) {blendSpace=/Game/Animation/Game/MainPlayer/Poses/Medium/Male/AO/M_NoWep_Relaxed_AO, alphaBoolBlend {blendOption=Linear}})
    Alpha=Get SurfaceSwimmingOffsetAlpha()
    Translation="0,0,-15"
    Rotation="0,0,0"
    Scale="1,1,1"
   ) {translationMode=BMM_Additive, alphaBoolBlend {blendOption=Linear}, boneToModify {boneName=Root}, translation {z=-15}}
   Output Pose(Result=Component To Local(ComponentPose=@n3)) {layerGroup=Item Override}
   Get TargetYaw()
  ]}
 ItemFullBodyAimOffsets {nodes [
   Input Pose!() -> (Pose, FullBody_AimOffsetAlpha:double, Aim_Yaw:double, Aim_Pitch:double) {name=InPoseFullBodyAimOffsets}
   Input Pose!() -> (Pose) {name=InPosePreFullBodyAimOffsets}
   Output Pose() {layerGroup=Item Override}
  ]}
 ItemFullBodyOverride {nodes [
   Input Pose() {name=InPoseFullBody}
   Output Pose() {layerGroup=Item Override}
  ]}
 ItemIdleOverride {nodes [
   Input Pose() {name=InPose_Idle}
   Output Pose() {layerGroup=Item Override}
  ]}
 ItemJetPackJumpAdditive {nodes [Output Pose(Result=Blend Poses by bool(
     BlendPose_0="Sequence Player (additive) 'JetpackApex_ARDS_Loop_Additive'"() {sequence=/Game/Animation/Game/MainPlayer/Locomotion/Medium/Male/Jump/AssaultRifle/JetpackApex_ARDS_Loop_Additive}
     BlendPose_1="Sequence Player (additive) 'JetpackApex_NoWep__Loop_CMM_Additive'"() {sequence=/Game/Animation/Game/MainPlayer/Locomotion/Medium/Male/Jump/NoWep/JetpackApex_NoWep__Loop_CMM_Additive}
     BlendTime_1=1
     bActiveValue=Get bIsMeleeGuarding()
    ) {blendType=Linear}) {layerGroup=Item Override}]}
 ItemJetPackStartAdditive {nodes [Output Pose() {layerGroup=Item Override}]}
 ItemJumpLoopAdditive {nodes [
   Input Pose() {name=InPoseItemJumpLoopAdditive}
   Output Pose(Result=Blend Poses by bool(
     BlendPose_0="Sequence Player (additive) 'M_JumpApex_ARDS_Loop_Additive'"() {playRate=0.5, sequence=/Game/Animation/Game/MainPlayer/Locomotion/Medium/Male/Jump/AssaultRifle/M_JumpApex_ARDS_Loop_Additive}
     BlendPose_1="Sequence Player (additive) 'JumpApex_NoWep__Loop_CMM_Additive'"() {sequence=/Game/Animation/Game/MainPlayer/Locomotion/Medium/Male/Jump/NoWep/JumpApex_NoWep__Loop_CMM_Additive}
     bActiveValue=Get bIsMeleeGuarding()
    ) {blendType=Linear}) {layerGroup=Item Override}
  ]}
 ItemJumpOverride {nodes [
   Input Pose() {name=InPoseJump}
   Output Pose() {layerGroup=Item Override}
  ]}
 ItemJumpUpAdditive {nodes [
   Input Pose() {name=InPoseItemJumpUpAdditive}
   Output Pose(Result=Blend Poses by bool(
     BlendPose_0="Sequence Player (additive) 'M_JumpUp_ARDS_Additive'"() {bLoopAnimation=false, sequence=/Game/Animation/Game/MainPlayer/Locomotion/Medium/Male/Jump/AssaultRifle/M_JumpUp_ARDS_Additive}
     BlendPose_1="Sequence Player (additive) 'JumpUp_NoWep_CMM_Additive'"() {bLoopAnimation=false, sequence=/Game/Animation/Game/MainPlayer/Locomotion/Medium/Male/Jump/NoWep/JumpUp_NoWep_CMM_Additive}
     bActiveValue=Get bIsMeleeGuarding()
    ) {blendType=Linear}) {layerGroup=Item Override}
  ]}
 ItemPreAimOffsets {nodes [
   Input Pose() {name=InPosePreAimOffsets}
   Output Pose() {layerGroup=Item Override}
  ]}
 ItemPreIK {nodes [
   Input Pose() {name=InPose_PreIK}
   Output Pose() {layerGroup=Item Override}
  ]}
 ItemRightHandIK {nodes [
   Input Pose() {name=InPosePreRightHandIK}
   Input Pose() {name=InPosePostRightHandIK}
   Output Pose() {layerGroup=Item Override}
  ]}
 ItemSkeletalControl {nodes [
   n1 = Local To Component(LocalPose=Input Pose())
   n2 = Get RootRotationOffset() -> (RootRotationOffset:Rotator)
   Output Pose(Result=Component To Local(ComponentPose="Transform (Modify) Bone - Bone: Root"!(ComponentPose=@n1, AlphaCurveName:Name=EnableAccelYawCorrection, Rotation=@n2) -> (Pose) {alphaInputType=Curve, rotationMode=BMM_Additive, alphaBoolBlend {blendOption=Linear}, boneToModify {boneName=Root}})) {layerGroup=Item Override}
  ]}
 ItemSwimJumpFallLoop {nodes [Output Pose() {layerGroup=Item Override}]}
 ItemSwimJumpSurfaceEnd {nodes [Output Pose() {layerGroup=Item Override}]}
 ItemSwimJumpSurfaceLoop {nodes [Output Pose() {layerGroup=Item Override}]}
 ItemUpperBody {nodes [
   AnimGraphNode_StateMachine_0 = Deflect() {is=title AnimGraphNode_StateMachine, bAllowConduitEntryStates=true}
   Output Pose(Result=@AnimGraphNode_StateMachine_0) {layerGroup=Item Override}
   Save cached pose 'UpperBody_CachePose'(Pose=Input Pose!() -> (Pose, InputParam_ResultingYaw:float=~, InputParam_ResultingPitch:float=~, InputParam_ZeroOutPitchWeightCurveName:Name) {name=InPoseUpperBody}) {cacheName=UpperBody_CachePose}
  ]}
 Deflect {
  entry=Conduit
  path=ItemUpperBody.AnimGraphNode_StateMachine_0.Deflect
  ids=id
  nodes [
   AnimStateConduitNode_1 = Conduit()
   AnimStateNode_10 = Deflect() {is=title AnimStateNode}
   AnimStateNode_13 = Default() {is=title AnimStateNode}
   AnimStateNode_9 = Deflect_to_Default() {is=title AnimStateNode}
   AnimStateTransitionNode_24 = AnimStateTransitionNode() {crossfadeDuration=0.35, logicType=TLT_Inertialization}
   AnimStateTransitionNode_25 = AnimStateTransitionNode() {crossfadeDuration=0.05, logicType=TLT_Inertialization}
   AnimStateTransitionNode_26 = AnimStateTransitionNode() {automaticRuleTriggerTime=0, crossfadeDuration=0.3, priorityOrder=100}
   AnimStateTransitionNode_27 = AnimStateTransitionNode()
   AnimStateTransitionNode_28 = AnimStateTransitionNode() {bAutomaticRuleBasedOnSequencePlayerInState=true}
   AnimStateTransitionNode_29 = AnimStateTransitionNode() {bAutomaticRuleBasedOnSequencePlayerInState=true}
   AnimStateTransitionNode_30 = AnimStateTransitionNode() {logicType=TLT_Inertialization}
  ]
  links [
   {
    dst=Default
    rule=ItemUpperBody.AnimGraphNode_StateMachine_0.Deflect.AnimStateTransitionNode_26.Transition
    src=Conduit
    via=AnimStateTransitionNode_26
   }
   {
    dst=Deflect
    rule=ItemUpperBody.AnimGraphNode_StateMachine_0.Deflect.AnimStateTransitionNode_25.Transition
    src=Conduit
    via=AnimStateTransitionNode_25
   }
   {
    dst=Conduit
    rule=ItemUpperBody.AnimGraphNode_StateMachine_0.Deflect.AnimStateTransitionNode_27.Transition
    src=Default
    via=AnimStateTransitionNode_27
   }
   {
    dst=Conduit
    rule=ItemUpperBody.AnimGraphNode_StateMachine_0.Deflect.AnimStateTransitionNode_28.Transition
    src=Deflect
    via=AnimStateTransitionNode_28
   }
   {
    dst=Deflect_to_Default
    rule=ItemUpperBody.AnimGraphNode_StateMachine_0.Deflect.AnimStateTransitionNode_24.Transition
    src=Deflect
    via=AnimStateTransitionNode_24
   }
   {
    dst=Conduit
    rule=ItemUpperBody.AnimGraphNode_StateMachine_0.Deflect.AnimStateTransitionNode_29.Transition
    src=Deflect_to_Default
    via=AnimStateTransitionNode_29
   }
   {
    dst=Deflect
    rule=ItemUpperBody.AnimGraphNode_StateMachine_0.Deflect.AnimStateTransitionNode_30.Transition
    src=Deflect_to_Default
    via=AnimStateTransitionNode_30
   }
  ]
 }
 Conduit {path=ItemUpperBody.AnimGraphNode_StateMachine_0.Deflect.AnimStateConduitNode_1.Conduit, nodes [Result(bCanEnterTransition=true)]}
 Deflect {
  path=ItemUpperBody.AnimGraphNode_StateMachine_0.Deflect.AnimStateNode_10.Deflect
  nodes [Output Animation Pose(Result=Sequence Player 'Realm_DeflectComboLoop'() {sequence=/Realm/Anims/Realm_DeflectComboLoop})]
 }
 Default {
  path=ItemUpperBody.AnimGraphNode_StateMachine_0.Deflect.AnimStateNode_13.Default
  nodes [Output Animation Pose(Result=Use cached pose 'UpperBody_CachePose'())]
 }
 Deflect_to_Default {
  path=ItemUpperBody.AnimGraphNode_StateMachine_0.Deflect.AnimStateNode_9.Deflect_to_Default
  nodes [
   AnimGraphNode_StateMachine_0 = DeflectExits() {is=title AnimGraphNode_StateMachine, bAllowConduitEntryStates=true}
   Output Animation Pose(Result=Apply Additive(
     Base=@AnimGraphNode_StateMachine_0
     Additive="Sequence Player (additive) 'Galileo_HeavyLamp_Lobster_Idle_Pose_Additive_CMM'"() {bLoopAnimation=false, sequence=/JethroLager/Assets/Animations/Galileo_HeavyLamp_Lobster_Idle_Pose_Additive_CMM}
     Alpha=Get EndPoseAlpha() -> (EndPoseAlpha:float)
    ) {alphaBoolBlend {blendOption=Linear}})
  ]
 }
 DeflectExits {
  entry=Conduit
  path=ItemUpperBody.AnimGraphNode_StateMachine_0.Deflect.AnimStateNode_9.Deflect_to_Default.AnimGraphNode_StateMachine_0.DeflectExits
  ids=id
  nodes [
   AnimStateConduitNode_0 = Conduit()
   AnimStateNode_0 = DeflectExit_1() {is=title AnimStateNode}
   AnimStateNode_1 = DeflectExit_2() {is=title AnimStateNode}
   AnimStateNode_2 = DeflectExit_3() {is=title AnimStateNode}
   AnimStateNode_3 = DeflectExit_4() {is=title AnimStateNode}
   AnimStateNode_4 = DeflectExit_5() {is=title AnimStateNode}
   AnimStateNode_5 = DeflectExit_6() {is=title AnimStateNode}
   AnimStateNode_6 = DeflectExit_7() {is=title AnimStateNode}
   AnimStateNode_7 = DeflectExit_Default() {is=title AnimStateNode}
   AnimStateTransitionNode_0 = AnimStateTransitionNode() {logicType=TLT_Inertialization}
   AnimStateTransitionNode_1 = AnimStateTransitionNode() {logicType=TLT_Inertialization, priorityOrder=2}
   AnimStateTransitionNode_2 = AnimStateTransitionNode() {logicType=TLT_Inertialization, priorityOrder=3}
   AnimStateTransitionNode_3 = AnimStateTransitionNode() {logicType=TLT_Inertialization, priorityOrder=4}
   AnimStateTransitionNode_4 = AnimStateTransitionNode() {logicType=TLT_Inertialization, priorityOrder=5}
   AnimStateTransitionNode_5 = AnimStateTransitionNode() {logicType=TLT_Inertialization, priorityOrder=6}
   AnimStateTransitionNode_6 = AnimStateTransitionNode() {logicType=TLT_Inertialization, priorityOrder=7}
   AnimStateTransitionNode_7 = AnimStateTransitionNode() {logicType=TLT_Inertialization, priorityOrder=100}
  ]
  links [
   {
    dst=DeflectExit_1
    rule=ItemUpperBody.AnimGraphNode_StateMachine_0.Deflect.AnimStateNode_9.Deflect_to_Default.AnimGraphNode_StateMachine_0.DeflectExits.AnimStateTransitionNode_0.Transition
    src=Conduit
    via=AnimStateTransitionNode_0
   }
   {
    dst=DeflectExit_2
    rule=ItemUpperBody.AnimGraphNode_StateMachine_0.Deflect.AnimStateNode_9.Deflect_to_Default.AnimGraphNode_StateMachine_0.DeflectExits.AnimStateTransitionNode_1.Transition
    src=Conduit
    via=AnimStateTransitionNode_1
   }
   {
    dst=DeflectExit_3
    rule=ItemUpperBody.AnimGraphNode_StateMachine_0.Deflect.AnimStateNode_9.Deflect_to_Default.AnimGraphNode_StateMachine_0.DeflectExits.AnimStateTransitionNode_2.Transition
    src=Conduit
    via=AnimStateTransitionNode_2
   }
   {
    dst=DeflectExit_4
    rule=ItemUpperBody.AnimGraphNode_StateMachine_0.Deflect.AnimStateNode_9.Deflect_to_Default.AnimGraphNode_StateMachine_0.DeflectExits.AnimStateTransitionNode_3.Transition
    src=Conduit
    via=AnimStateTransitionNode_3
   }
   {
    dst=DeflectExit_5
    rule=ItemUpperBody.AnimGraphNode_StateMachine_0.Deflect.AnimStateNode_9.Deflect_to_Default.AnimGraphNode_StateMachine_0.DeflectExits.AnimStateTransitionNode_4.Transition
    src=Conduit
    via=AnimStateTransitionNode_4
   }
   {
    dst=DeflectExit_6
    rule=ItemUpperBody.AnimGraphNode_StateMachine_0.Deflect.AnimStateNode_9.Deflect_to_Default.AnimGraphNode_StateMachine_0.DeflectExits.AnimStateTransitionNode_5.Transition
    src=Conduit
    via=AnimStateTransitionNode_5
   }
   {
    dst=DeflectExit_7
    rule=ItemUpperBody.AnimGraphNode_StateMachine_0.Deflect.AnimStateNode_9.Deflect_to_Default.AnimGraphNode_StateMachine_0.DeflectExits.AnimStateTransitionNode_6.Transition
    src=Conduit
    via=AnimStateTransitionNode_6
   }
   {
    dst=DeflectExit_Default
    rule=ItemUpperBody.AnimGraphNode_StateMachine_0.Deflect.AnimStateNode_9.Deflect_to_Default.AnimGraphNode_StateMachine_0.DeflectExits.AnimStateTransitionNode_7.Transition
    src=Conduit
    via=AnimStateTransitionNode_7
   }
  ]
 }
 Conduit {
  path=ItemUpperBody.AnimGraphNode_StateMachine_0.Deflect.AnimStateNode_9.Deflect_to_Default.AnimGraphNode_StateMachine_0.DeflectExits.AnimStateConduitNode_0.Conduit
  nodes [Result(bCanEnterTransition=true)]
 }
 DeflectExit_1 {
  path=ItemUpperBody.AnimGraphNode_StateMachine_0.Deflect.AnimStateNode_9.Deflect_to_Default.AnimGraphNode_StateMachine_0.DeflectExits.AnimStateNode_0.DeflectExit_1
  nodes [Output Animation Pose(Result=Sequence Player 'Realm_Deflect1_Exit'() {bLoopAnimation=false, sequence=/Realm/Anims/Realm_Deflect1_Exit})]
 }
 DeflectExit_2 {
  path=ItemUpperBody.AnimGraphNode_StateMachine_0.Deflect.AnimStateNode_9.Deflect_to_Default.AnimGraphNode_StateMachine_0.DeflectExits.AnimStateNode_1.DeflectExit_2
  nodes [Output Animation Pose(Result=Sequence Player 'Realm_Deflect2_Exit'() {bLoopAnimation=false, sequence=/Realm/Anims/Realm_Deflect2_Exit})]
 }
 DeflectExit_3 {
  path=ItemUpperBody.AnimGraphNode_StateMachine_0.Deflect.AnimStateNode_9.Deflect_to_Default.AnimGraphNode_StateMachine_0.DeflectExits.AnimStateNode_2.DeflectExit_3
  nodes [Output Animation Pose(Result=Sequence Player 'Realm_Deflect3_Exit'() {bLoopAnimation=false, sequence=/Realm/Anims/Realm_Deflect3_Exit})]
 }
 DeflectExit_4 {
  path=ItemUpperBody.AnimGraphNode_StateMachine_0.Deflect.AnimStateNode_9.Deflect_to_Default.AnimGraphNode_StateMachine_0.DeflectExits.AnimStateNode_3.DeflectExit_4
  nodes [Output Animation Pose(Result=Sequence Player 'Realm_Deflect4_Exit'() {bLoopAnimation=false, sequence=/Realm/Anims/Realm_Deflect4_Exit})]
 }
 DeflectExit_5 {
  path=ItemUpperBody.AnimGraphNode_StateMachine_0.Deflect.AnimStateNode_9.Deflect_to_Default.AnimGraphNode_StateMachine_0.DeflectExits.AnimStateNode_4.DeflectExit_5
  nodes [Output Animation Pose(Result=Sequence Player 'Realm_Deflect5_Exit'() {bLoopAnimation=false, sequence=/Realm/Anims/Realm_Deflect5_Exit})]
 }
 DeflectExit_6 {
  path=ItemUpperBody.AnimGraphNode_StateMachine_0.Deflect.AnimStateNode_9.Deflect_to_Default.AnimGraphNode_StateMachine_0.DeflectExits.AnimStateNode_5.DeflectExit_6
  nodes [Output Animation Pose(Result=Sequence Player 'Realm_Deflect6_Exit'() {bLoopAnimation=false, sequence=/Realm/Anims/Realm_Deflect6_Exit})]
 }
 DeflectExit_7 {
  path=ItemUpperBody.AnimGraphNode_StateMachine_0.Deflect.AnimStateNode_9.Deflect_to_Default.AnimGraphNode_StateMachine_0.DeflectExits.AnimStateNode_6.DeflectExit_7
  nodes [Output Animation Pose(Result=Sequence Player 'Realm_Deflect7_Exit'() {bLoopAnimation=false, sequence=/Realm/Anims/Realm_Deflect7_Exit})]
 }
 DeflectExit_Default {
  path=ItemUpperBody.AnimGraphNode_StateMachine_0.Deflect.AnimStateNode_9.Deflect_to_Default.AnimGraphNode_StateMachine_0.DeflectExits.AnimStateNode_7.DeflectExit_Default
  nodes [Output Animation Pose(Result=Sequence Player 'Realm_Deflect8_Exit'() {bLoopAnimation=false, sequence=/Realm/Anims/Realm_Deflect8_Exit})]
 }
 Transition {
  path=ItemUpperBody.AnimGraphNode_StateMachine_0.Deflect.AnimStateNode_9.Deflect_to_Default.AnimGraphNode_StateMachine_0.DeflectExits.AnimStateTransitionNode_0.Transition
  nodes [Result(bCanEnterTransition=Get bExit_1() -> (bExit_1:bool))]
 }
 Transition {
  path=ItemUpperBody.AnimGraphNode_StateMachine_0.Deflect.AnimStateNode_9.Deflect_to_Default.AnimGraphNode_StateMachine_0.DeflectExits.AnimStateTransitionNode_1.Transition
  nodes [Result(bCanEnterTransition=Get bExit_2() -> (bExit_2:bool))]
 }
 Transition {
  path=ItemUpperBody.AnimGraphNode_StateMachine_0.Deflect.AnimStateNode_9.Deflect_to_Default.AnimGraphNode_StateMachine_0.DeflectExits.AnimStateTransitionNode_2.Transition
  nodes [Result(bCanEnterTransition=Get bExit_3() -> (bExit_3:bool))]
 }
 Transition {
  path=ItemUpperBody.AnimGraphNode_StateMachine_0.Deflect.AnimStateNode_9.Deflect_to_Default.AnimGraphNode_StateMachine_0.DeflectExits.AnimStateTransitionNode_3.Transition
  nodes [Result(bCanEnterTransition=Get bExit_4() -> (bExit_4:bool))]
 }
 Transition {
  path=ItemUpperBody.AnimGraphNode_StateMachine_0.Deflect.AnimStateNode_9.Deflect_to_Default.AnimGraphNode_StateMachine_0.DeflectExits.AnimStateTransitionNode_4.Transition
  nodes [Result(bCanEnterTransition=Get bExit_5() -> (bExit_5:bool))]
 }
 Transition {
  path=ItemUpperBody.AnimGraphNode_StateMachine_0.Deflect.AnimStateNode_9.Deflect_to_Default.AnimGraphNode_StateMachine_0.DeflectExits.AnimStateTransitionNode_5.Transition
  nodes [Result(bCanEnterTransition=Get bExit_6() -> (bExit_6:bool))]
 }
 Transition {
  path=ItemUpperBody.AnimGraphNode_StateMachine_0.Deflect.AnimStateNode_9.Deflect_to_Default.AnimGraphNode_StateMachine_0.DeflectExits.AnimStateTransitionNode_6.Transition
  nodes [Result(bCanEnterTransition=Get bExit_7() -> (bExit_7:bool))]
 }
 Transition {
  path=ItemUpperBody.AnimGraphNode_StateMachine_0.Deflect.AnimStateNode_9.Deflect_to_Default.AnimGraphNode_StateMachine_0.DeflectExits.AnimStateTransitionNode_7.Transition
  nodes [Result(bCanEnterTransition=true)]
 }
 Transition {
  path=ItemUpperBody.AnimGraphNode_StateMachine_0.Deflect.AnimStateTransitionNode_24.Transition
  nodes [Result(bCanEnterTransition=NOT Boolean(A=Get bIsCurrentlyDeflecting() -> (bIsCurrentlyDeflecting:bool)))]
 }
 Transition {
  path=ItemUpperBody.AnimGraphNode_StateMachine_0.Deflect.AnimStateTransitionNode_25.Transition
  nodes [Result(bCanEnterTransition=Get bIsCurrentlyDeflecting())]
 }
 Transition {path=ItemUpperBody.AnimGraphNode_StateMachine_0.Deflect.AnimStateTransitionNode_26.Transition, nodes [Result(bCanEnterTransition=true)]}
 Transition {path=ItemUpperBody.AnimGraphNode_StateMachine_0.Deflect.AnimStateTransitionNode_27.Transition, nodes [Result(bCanEnterTransition=true)]}
 Transition {path=ItemUpperBody.AnimGraphNode_StateMachine_0.Deflect.AnimStateTransitionNode_28.Transition, nodes [Result(bCanEnterTransition=true)]}
 Transition {path=ItemUpperBody.AnimGraphNode_StateMachine_0.Deflect.AnimStateTransitionNode_29.Transition, nodes [Result()]}
 Transition {
  path=ItemUpperBody.AnimGraphNode_StateMachine_0.Deflect.AnimStateTransitionNode_30.Transition
  nodes [Result(bCanEnterTransition=Get bIsCurrentlyDeflecting())]
 }
 ItemZipLineStartAdditive {nodes [Output Pose(Result=Blend Poses by bool(
     BlendPose_0="Sequence Player (additive) 'ZipLine_Entry_ARDS_Additive'"() {bLoopAnimation=false, sequence=/Game/Animation/Game/MainPlayer/Combat/Gadgets/Medium/Male/ZipLine/ZipLine_Entry_ARDS_Additive}
     BlendPose_1="Sequence Player (additive) 'ZipLine_Entry_NoWep_Additive'"() {bLoopAnimation=false, sequence=/Game/Animation/Game/MainPlayer/Combat/Gadgets/Medium/Male/ZipLine/ZipLine_Entry_NoWep_Additive}
     bActiveValue=Get bIsMeleeGuarding()
    ) {blendType=Linear}) {layerGroup=Item Override}]}
]
variables [
 Delta Time X {type=number}
 SurfaceSwimmingOffsetAlpha {type=number}
 Temp_Replacement_Yaw {type=number}
]
