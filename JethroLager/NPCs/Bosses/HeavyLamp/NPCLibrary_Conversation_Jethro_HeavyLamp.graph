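# Conversation database for the Jethro HeavyLamp boss NPC, which asks whether the player is ready to be taught.
# Its single tagged entry point opens a speech that branches to a Back node for declining or a second speech for accepting.
# Accepting leads into a HeavyLamp-specific task node that, by its name, spawns a rift gate.
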ConversationDatabase /JethroLager/NPCs/Bosses/HeavyLamp/NPCLibrary_Conversation_Jethro_HeavyLamp
entry_points [{tag=Conversation.Entry.Jethro.HeavyLamp, destinations [#1]}]
nodes [
 ConversationEntryPointNode_1 {
  id=#1
  class=ConversationEntryPointNode
  outputs [#2]
  properties {entryTag {tagName=Conversation.Entry.Jethro.HeavyLamp}}
 }
 Speech_C_9 {
  id=#2
  class=Speech_C
  outputs [#3, #4]
  properties {
   nodeName=""
   softTaskIcon=&T_UI_NPC_question
   generalConfig {defaultMessage=Are you ready for what I have to teach?}
  }
  sub_nodes [ConversationChoiceNode_14 {class=ConversationChoiceNode, properties {defaultChoiceDisplayText=The Force is strong here.}}]
 }
 Speech_C_10 {
  id=#4
  class=Speech_C
  outputs [#5]
  properties {nodeName="", softTaskIcon=&T_UI_NPC_question, generalConfig {defaultMessage=Then let us begin.}}
  sub_nodes [ConversationChoiceNode_24 {class=ConversationChoiceNode, properties {defaultChoiceDisplayText="I am ready to learn, my master."}}]
 }
 FCTN_HeavyLamp_SpawnRiftGate_C_0 {id=#5, class=FCTN_HeavyLamp_SpawnRiftGate_C, outputs []}
 Back_C_2 {
  id=#3
  class=Back_C
  sub_nodes [ConversationChoiceNode_15 {class=ConversationChoiceNode, properties {defaultChoiceDisplayText=No. I am not ready.}}]
 }
]
